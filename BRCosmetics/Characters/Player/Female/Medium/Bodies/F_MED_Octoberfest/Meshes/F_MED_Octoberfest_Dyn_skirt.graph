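# Physics asset simulating the skirt and rear bow of the female medium Octoberfest cosmetic body, with capsules for pelvis, lower spine and legs.
# Eight box-and-capsule skirt chains hang from the pelvis with angular limits, front and side ones via small MSTR bodies.
# Bow loops hang off a spine-attached base, while its strings hang from the center back panel.
# Angularly free constraints cross-link most neighboring front and side lower segments, plus back-left to left side.
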PhysicsAsset /BRCosmetics/Characters/Player/Female/Medium/Bodies/F_MED_Octoberfest/Meshes/F_MED_Octoberfest_Dyn_skirt
bodies {
 calf_l [{shapeType=Capsule, length=29.3752, radius=5.4592, center=<-14.0706 -3.5012 -0.4484>, rotation=<rot 0 100.888 -85.8593>}]
 calf_r [{shapeType=Capsule, length=29.3752, radius=5.4592, center=<14.0706 3.5012 0.4484>, rotation=<rot 0 100.888 -85.8593>}]
 dyn_buttbow_base [{shapeType=Capsule, length=1, radius=1}]
 dyn_buttbow_loop_l [{shapeType=Capsule, length=3.5474, radius=2.7104, center=<0 0 -4>}]
 dyn_buttbow_loop_r [{shapeType=Capsule, length=3.5474, radius=2.7104, center=<0 0 -4>}]
 dyn_buttbow_string_l_01 [{shapeType=Capsule, length=2.2922, radius=1.6303, center=<1.1726 -0.0002 -1.863>, rotation=<rot -22.6946 0 0>}]
 dyn_buttbow_string_l_02 [{shapeType=Capsule, length=2.4723, radius=1.6303, center=<-0.5901 0 -1>, rotation=<rot -19.9331 0 0>}]
 dyn_buttbow_string_r_01 [{shapeType=Capsule, length=2.2922, radius=1.6303, center=<1.0659 0.5171 -1.7541>, rotation=<rot -22.6946 0 0>}]
 dyn_buttbow_string_r_02 [{shapeType=Capsule, length=2.4723, radius=1.6303, center=<-0.5901 0 -1>, rotation=<rot -19.9331 0 0>}]
 dyn_skirt_bk_01 [{shapeType=Box, extentX=15.1888, extentY=16.8437, extentZ=1.3193, center=<6.2207 0 0>}]
 dyn_skirt_bk_02 [{shapeType=Capsule, length=8.5, radius=4.5, shapeName=dyn_skirt_bk_02_capsule, center=<7.8196 0 0>, rotation=<rot 0 0 90>}]
 dyn_skirt_bk_03 [{shapeType=Capsule, length=10.5, radius=4.5, shapeName=dyn_skirt_bk_03_capsule, center=<5.1451 0 -0.0001>, rotation=<rot 0 0 90>}]
 dyn_skirt_bk_04 [{shapeType=Capsule, length=7, radius=1.5, shapeName=dyn_skirt_bk_04_capsule, center=<5 0 0>, rotation=<rot 90 0 0>}]
 dyn_skirt_bk_l_01 [{shapeType=Box, extentX=15.1888, extentY=12.5058, extentZ=1.3193, center=<7.8196 0 0>, rotation=<rot 0 0 -3.7044>}]
 dyn_skirt_bk_l_02 [{shapeType=Capsule, length=20.5, radius=4.5, shapeName=dyn_skirt_bk_l_02_capsule, center=<7.8196 0 0>, rotation=<rot 0 0 90>}]
 dyn_skirt_bk_l_03 [{shapeType=Capsule, length=23.5, radius=4.5, shapeName=dyn_skirt_bk_l_03_capsule, center=<3 0 0>, rotation=<rot 0 0 90>}]
 dyn_skirt_bk_l_04 [{shapeType=Capsule, length=7, radius=1, shapeName=dyn_skirt_bk_l_04_capsule, center=<5 0 0>, rotation=<rot 90 0 0>}]
 dyn_skirt_bk_r_01 [{shapeType=Box, extentX=15.1888, extentY=12.5058, extentZ=1.3193, center=<7.8196 0 1>, rotation=<rot 0 0.0002 -1.2603>}]
 dyn_skirt_bk_r_02 [{shapeType=Capsule, length=27.5, radius=4.5, shapeName=dyn_skirt_bk_r_02_capsule, center=<7.8196 0 0>, rotation=<rot 0 0 90>}]
 dyn_skirt_bk_r_03 [{shapeType=Capsule, length=23.5, radius=4.5, shapeName=dyn_skirt_bk_r_03_capsule, center=<3 0 0>, rotation=<rot 0 0 90>}]
 dyn_skirt_bk_r_04 [{shapeType=Capsule, length=7.1295, radius=1.8705, shapeName=dyn_skirt_bk_r_04_capsule, center=<3.6664 0 1.2173>, rotation=<rot 90 0 0>}]
 dyn_skirt_fr_01 [{shapeType=Box, extentX=11, extentY=12, extentZ=3.5, shapeName=dyn_skirt_fr_01_box, center=<10 0 2>}]
 dyn_skirt_fr_02 [{shapeType=Capsule, length=17.5, radius=3.5, shapeName=dyn_skirt_fr_02_capsule, center=<7.8196 0 0>, rotation=<rot 0 0 90>}]
 dyn_skirt_fr_03 [{shapeType=Capsule, length=14.5, radius=3.7, shapeName=dyn_skirt_fr_03_capsule, center=<3 0 0>, rotation=<rot 0 0 90>}]
 dyn_skirt_fr_04 [{shapeType=Capsule, length=7, radius=2, shapeName=dyn_skirt_fr_04_capsule, center=<5 0 0>, rotation=<rot 90 0 0>}]
 dyn_skirt_fr_MSTR [{shapeType=Capsule, length=1, radius=1}]
 dyn_skirt_fr_l_01 [{shapeType=Box, extentX=11.2946, extentY=13.158, extentZ=1.6318, center=<11.9251 0.219 0.4911>, rotation=<rot 5 0 0>}]
 dyn_skirt_fr_l_02 [{shapeType=Capsule, length=15.75, radius=4.25, shapeName=dyn_skirt_fr_l_02_capsule, center=<4 0 0>, rotation=<rot 0 0 90>}]
 dyn_skirt_fr_l_03 [{shapeType=Capsule, length=21.25, radius=4, shapeName=dyn_skirt_fr_l_03_capsule, center=<4 -1 0>, rotation=<rot 0 0 90>}]
 dyn_skirt_fr_l_04 [{shapeType=Capsule, length=7, radius=2, shapeName=dyn_skirt_fr_l_04_capsule, center=<5 0 0>, rotation=<rot 90 0 0>}]
 dyn_skirt_fr_l_MSTR [{shapeType=Capsule, length=1, radius=1}]
 dyn_skirt_fr_r_01 [{shapeType=Box, extentX=11.2946, extentY=12.658, extentZ=1.6318, center=<11.9527 0 1>, rotation=<rot 5 0 0>}]
 dyn_skirt_fr_r_02 [{shapeType=Capsule, length=14.5, radius=3.5, shapeName=dyn_skirt_fr_r_02_capsule, center=<7.8196 0 1>, rotation=<rot 0 0 90>}]
 dyn_skirt_fr_r_03 [{shapeType=Capsule, length=12.25, radius=4.75, shapeName=dyn_skirt_fr_r_03_capsule, center=<3 0 1>, rotation=<rot 0 0 90>}]
 dyn_skirt_fr_r_04 [{shapeType=Capsule, length=7, radius=2, shapeName=dyn_skirt_fr_r_04_capsule, center=<5 0 0>, rotation=<rot 90 0 0>}]
 dyn_skirt_fr_r_MSTR [{shapeType=Capsule, length=1, radius=1}]
 dyn_skirt_out_l_01 [{shapeType=Box, extentX=15.1888, extentY=11.5326, extentZ=3.0068, center=<7.8196 0 0>}]
 dyn_skirt_out_l_02 [{shapeType=Capsule, length=18.75, radius=3.5, shapeName=dyn_skirt_out_l_02_capsule, center=<7.8196 0 0>, rotation=<rot 0 0 90>}]
 dyn_skirt_out_l_03 [{shapeType=Capsule, length=30, radius=4, shapeName=dyn_skirt_out_l_03_capsule, center=<3 0 0>, rotation=<rot 0 0 90>}]
 dyn_skirt_out_l_04 [{shapeType=Capsule, length=7, radius=1, shapeName=dyn_skirt_out_l_04_capsule, center=<5 0 0>, rotation=<rot 90 0 0>}]
 dyn_skirt_out_l_MSTR [{shapeType=Capsule, length=1, radius=1}]
 dyn_skirt_out_r_01 [{shapeType=Box, extentX=15.1888, extentY=9.6924, extentZ=3.0068, center=<7.8196 0 0>, rotation=<rot 0 0 13.3429>}]
 dyn_skirt_out_r_02 [{shapeType=Capsule, length=13.5, radius=5.5, shapeName=dyn_skirt_out_r_02_capsule, center=<7.8196 0 0>, rotation=<rot 0 0 90>}]
 dyn_skirt_out_r_03 [{shapeType=Capsule, length=18.5, radius=4, shapeName=dyn_skirt_out_r_03_capsule, center=<3 0 0>, rotation=<rot 0 0 90>}]
 dyn_skirt_out_r_04 [{shapeType=Capsule, length=7, radius=2, shapeName=dyn_skirt_out_r_04_capsule, center=<5 0 0>, rotation=<rot 90 0 0>}]
 dyn_skirt_out_r_MSTR [{shapeType=Capsule, length=1, radius=1}]
 pelvis [{shapeType=Capsule, length=9.7736, radius=9.4985, center=<-5.7764 -0.9503 0>}]
 spine_01 [{shapeType=Capsule, length=8.5067, radius=8.0118, center=<-5.7764 0.1541 0>}]
 thigh_l [{shapeType=Capsule, length=37.0283, radius=8.3282, center=<-25.6131 0.1862 0.7194>, rotation=<rot 86.5422 179.6205 178.886>}]
 thigh_r [{shapeType=Capsule, length=27.2968, radius=8.3282, center=<21.3088 -0.2106 -0.9794>, rotation=<rot 86.5422 179.6205 178.886>}]
}
constraints [
 {
  bone1Name=calf_l
  bone2Name=thigh_l
  swing1LimitDegrees=45
  swing1Motion=Free
  swing2LimitDegrees=45
  swing2Motion=Free
  twistLimitDegrees=45
  twistMotion=Free
 }
 {
  bone1Name=calf_r
  bone2Name=thigh_r
  swing1LimitDegrees=45
  swing1Motion=Free
  swing2LimitDegrees=45
  swing2Motion=Free
  twistLimitDegrees=45
  twistMotion=Free
 }
 {
  bone1Name=thigh_l
  bone2Name=pelvis
  swing1LimitDegrees=45
  swing1Motion=Free
  swing2LimitDegrees=45
  swing2Motion=Free
  twistLimitDegrees=45
  twistMotion=Free
 }
 {
  bone1Name=thigh_r
  bone2Name=pelvis
  swing1LimitDegrees=45
  swing1Motion=Free
  swing2LimitDegrees=45
  swing2Motion=Free
  twistLimitDegrees=45
  twistMotion=Free
 }
 {bone1Name=dyn_buttbow_loop_l, bone2Name=dyn_buttbow_base, swing1LimitDegrees=30, swing2LimitDegrees=45, twistLimitDegrees=10}
 {bone1Name=dyn_buttbow_loop_r, bone2Name=dyn_buttbow_base, swing1LimitDegrees=30, swing2LimitDegrees=45, twistLimitDegrees=10}
 {bone1Name=dyn_buttbow_string_r_01, bone2Name=dyn_skirt_bk_01, swing1LimitDegrees=60, swing2LimitDegrees=5, twistLimitDegrees=10}
 {bone1Name=dyn_buttbow_string_r_02, bone2Name=dyn_buttbow_string_r_01, swing1LimitDegrees=60, swing2LimitDegrees=90, twistLimitDegrees=10}
 {bone1Name=dyn_buttbow_string_l_01, bone2Name=dyn_skirt_bk_01, swing1LimitDegrees=60, swing2LimitDegrees=5, twistLimitDegrees=10}
 {bone1Name=dyn_buttbow_string_l_02, bone2Name=dyn_buttbow_string_l_01, swing1LimitDegrees=60, swing2LimitDegrees=90, twistLimitDegrees=10}
 {
  bone1Name=dyn_buttbow_base
  bone2Name=spine_01
  swing1LimitDegrees=45
  swing1Motion=Free
  swing2LimitDegrees=45
  swing2Motion=Free
  twistLimitDegrees=45
  twistMotion=Free
 }
 {bone1Name=dyn_skirt_fr_01, bone2Name=dyn_skirt_fr_MSTR, swing1LimitDegrees=45, swing2LimitDegrees=45, twistLimitDegrees=1}
 {bone1Name=dyn_skirt_fr_02, bone2Name=dyn_skirt_fr_01, swing1LimitDegrees=45, swing2LimitDegrees=45, twistLimitDegrees=5}
 {bone1Name=dyn_skirt_fr_03, bone2Name=dyn_skirt_fr_02, swing1LimitDegrees=30, swing2LimitDegrees=45, twistLimitDegrees=0.1}
 {bone1Name=dyn_skirt_fr_04, bone2Name=dyn_skirt_fr_02, swing1LimitDegrees=0.1, swing2LimitDegrees=45, twistLimitDegrees=0.1}
 {bone1Name=dyn_skirt_fr_l_01, bone2Name=dyn_skirt_fr_l_MSTR, swing1LimitDegrees=0.5, swing2LimitDegrees=45, twistLimitDegrees=0.5}
 {bone1Name=dyn_skirt_fr_l_02, bone2Name=dyn_skirt_fr_l_01, swing1LimitDegrees=0.5, swing2LimitDegrees=45, twistLimitDegrees=0.5}
 {bone1Name=dyn_skirt_fr_l_03, bone2Name=dyn_skirt_fr_l_02, swing1LimitDegrees=0.1, swing2LimitDegrees=20, twistLimitDegrees=0.1}
 {bone1Name=dyn_skirt_fr_l_04, bone2Name=dyn_skirt_fr_l_02, swing1LimitDegrees=0.1, swing2LimitDegrees=45, twistLimitDegrees=0.1}
 {bone1Name=dyn_skirt_out_l_01, bone2Name=dyn_skirt_out_l_MSTR, swing1LimitDegrees=1, swing2LimitDegrees=40, twistLimitDegrees=1}
 {
  bone1Name=dyn_skirt_out_l_02
  bone2Name=dyn_skirt_out_l_01
  swing1LimitDegrees=5
  swing1Motion=Locked
  swing2LimitDegrees=45
  twistLimitDegrees=5
 }
 {bone1Name=dyn_skirt_out_l_03, bone2Name=dyn_skirt_out_l_02, swing1LimitDegrees=5, swing2LimitDegrees=45, twistLimitDegrees=0.1}
 {bone1Name=dyn_skirt_out_l_04, bone2Name=dyn_skirt_out_l_02, swing1LimitDegrees=0.1, swing2LimitDegrees=45, twistLimitDegrees=0.1}
 {bone1Name=dyn_skirt_bk_l_01, bone2Name=pelvis, swing1LimitDegrees=90, swing2LimitDegrees=90, twistLimitDegrees=5}
 {bone1Name=dyn_skirt_bk_l_02, bone2Name=dyn_skirt_bk_l_01, swing1LimitDegrees=5, swing2LimitDegrees=80, twistLimitDegrees=5}
 {bone1Name=dyn_skirt_bk_l_03, bone2Name=dyn_skirt_bk_l_02, swing1LimitDegrees=5, swing2LimitDegrees=45, twistLimitDegrees=0.1}
 {bone1Name=dyn_skirt_bk_l_04, bone2Name=dyn_skirt_bk_l_02, swing1LimitDegrees=0.1, swing2LimitDegrees=45, twistLimitDegrees=0.1}
 {bone1Name=dyn_skirt_bk_01, bone2Name=pelvis, swing1LimitDegrees=45, swing2LimitDegrees=45, twistLimitDegrees=1}
 {bone1Name=dyn_skirt_bk_02, bone2Name=dyn_skirt_bk_01, swing1LimitDegrees=45, swing2LimitDegrees=80, twistLimitDegrees=5}
 {bone1Name=dyn_skirt_bk_03, bone2Name=dyn_skirt_bk_02, swing1LimitDegrees=45, swing2LimitDegrees=45, twistLimitDegrees=0.1}
 {bone1Name=dyn_skirt_bk_04, bone2Name=dyn_skirt_bk_02, swing1LimitDegrees=0.1, swing2LimitDegrees=45, twistLimitDegrees=0.1}
 {
  bone1Name=dyn_skirt_fr_r_MSTR
  bone2Name=pelvis
  swing1LimitDegrees=45
  swing1Motion=Free
  swing2LimitDegrees=45
  swing2Motion=Free
  twistLimitDegrees=45
  twistMotion=Free
 }
 {bone1Name=dyn_skirt_fr_r_01, bone2Name=dyn_skirt_fr_r_MSTR, swing1LimitDegrees=0.01, swing2LimitDegrees=45, twistLimitDegrees=1}
 {bone1Name=dyn_skirt_fr_r_02, bone2Name=dyn_skirt_fr_r_01, swing1LimitDegrees=0.01, swing2LimitDegrees=45, twistLimitDegrees=1}
 {bone1Name=dyn_skirt_fr_r_03, bone2Name=dyn_skirt_fr_r_02, swing1LimitDegrees=2, swing2LimitDegrees=45, twistLimitDegrees=0.1}
 {bone1Name=dyn_skirt_fr_r_04, bone2Name=dyn_skirt_fr_r_02, swing1LimitDegrees=0.1, swing2LimitDegrees=45, twistLimitDegrees=0.1}
 {
  bone1Name=dyn_skirt_out_r_MSTR
  bone2Name=pelvis
  swing1LimitDegrees=45
  swing1Motion=Free
  swing2LimitDegrees=45
  swing2Motion=Free
  twistLimitDegrees=45
  twistMotion=Free
 }
 {bone1Name=dyn_skirt_out_r_01, bone2Name=dyn_skirt_out_r_MSTR, swing1LimitDegrees=15, swing2LimitDegrees=60, twistLimitDegrees=1}
 {bone1Name=dyn_skirt_out_r_02, bone2Name=dyn_skirt_out_r_01, swing1LimitDegrees=5, swing2LimitDegrees=45, twistLimitDegrees=5}
 {bone1Name=dyn_skirt_out_r_04, bone2Name=dyn_skirt_out_r_02, swing1LimitDegrees=0.1, swing2LimitDegrees=45, twistLimitDegrees=0.1}
 {bone1Name=dyn_skirt_out_r_03, bone2Name=dyn_skirt_out_r_02, swing1LimitDegrees=5, swing2LimitDegrees=45, twistLimitDegrees=0.1}
 {bone1Name=dyn_skirt_bk_r_01, bone2Name=pelvis, swing1LimitDegrees=90, swing2LimitDegrees=90, twistLimitDegrees=5}
 {bone1Name=dyn_skirt_bk_r_02, bone2Name=dyn_skirt_bk_r_01, swing1LimitDegrees=5, swing2LimitDegrees=80, twistLimitDegrees=1}
 {bone1Name=dyn_skirt_bk_r_03, bone2Name=dyn_skirt_bk_r_02, swing1LimitDegrees=5, swing2LimitDegrees=45, twistLimitDegrees=0.1}
 {bone1Name=dyn_skirt_bk_r_04, bone2Name=dyn_skirt_bk_r_02, swing1LimitDegrees=0.1, swing2LimitDegrees=45, twistLimitDegrees=0.1}
 {
  bone1Name=dyn_skirt_fr_MSTR
  bone2Name=pelvis
  swing1LimitDegrees=45
  swing1Motion=Free
  swing2LimitDegrees=45
  swing2Motion=Free
  twistLimitDegrees=45
  twistMotion=Free
 }
 {
  bone1Name=dyn_skirt_fr_l_MSTR
  bone2Name=pelvis
  swing1LimitDegrees=45
  swing1Motion=Free
  swing2LimitDegrees=45
  swing2Motion=Free
  twistLimitDegrees=45
  twistMotion=Free
 }
 {
  bone1Name=dyn_skirt_out_l_MSTR
  bone2Name=pelvis
  swing1LimitDegrees=45
  swing1Motion=Free
  swing2LimitDegrees=45
  swing2Motion=Free
  twistLimitDegrees=45
  twistMotion=Free
 }
 {
  bone1Name=dyn_skirt_out_l_03
  bone2Name=dyn_skirt_fr_l_03
  swing1LimitDegrees=45
  swing1Motion=Free
  swing2LimitDegrees=45
  swing2Motion=Free
  twistLimitDegrees=45
  twistMotion=Free
 }
 {
  bone1Name=dyn_skirt_fr_03
  bone2Name=dyn_skirt_fr_l_03
  swing1LimitDegrees=45
  swing1Motion=Free
  swing2LimitDegrees=45
  swing2Motion=Free
  twistLimitDegrees=45
  twistMotion=Free
 }
 {
  bone1Name=dyn_skirt_out_r_03
  bone2Name=dyn_skirt_fr_r_03
  swing1LimitDegrees=45
  swing1Motion=Free
  swing2LimitDegrees=45
  swing2Motion=Free
  twistLimitDegrees=45
  twistMotion=Free
 }
 {
  bone1Name=dyn_skirt_fr_03
  bone2Name=dyn_skirt_fr_r_03
  swing1LimitDegrees=45
  swing1Motion=Free
  swing2LimitDegrees=45
  swing2Motion=Free
  twistLimitDegrees=45
  twistMotion=Free
 }
 {
  bone1Name=dyn_skirt_bk_l_03
  bone2Name=dyn_skirt_out_l_03
  swing1LimitDegrees=45
  swing1Motion=Free
  swing2LimitDegrees=45
  swing2Motion=Free
  twistLimitDegrees=45
  twistMotion=Free
 }
 {
  bone1Name=spine_01
  bone2Name=pelvis
  swing1LimitDegrees=45
  swing1Motion=Free
  swing2LimitDegrees=45
  swing2Motion=Free
  twistLimitDegrees=45
  twistMotion=Free
 }
]
