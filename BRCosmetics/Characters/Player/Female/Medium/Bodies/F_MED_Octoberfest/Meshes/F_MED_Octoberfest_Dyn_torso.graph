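# Physics asset for the torso of the female medium Octoberfest body cosmetic, covering its dangling chest bow, belly bow and choker bones.
# Bow loops and strings hang off base bodies with limited swing and tight twist, with belly strings as two-link chains; choker ends attach directly to spine_05.
# Chest and belly bow bases join spine_05 and spine_03 with free angular motion, as does the pelvis-spine-neck chain of larger body shapes.
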PhysicsAsset /BRCosmetics/Characters/Player/Female/Medium/Bodies/F_MED_Octoberfest/Meshes/F_MED_Octoberfest_Dyn_torso
bodies {
 dyn_bellybow_base [{shapeType=Capsule, length=1, radius=1}]
 dyn_bellybow_loop_l [{shapeType=Capsule, length=3.5474, radius=2.7104, center=<0 0 -4>}]
 dyn_bellybow_loop_r [{shapeType=Capsule, length=3.5474, radius=2.7104, center=<0 0 -4>}]
 dyn_bellybow_string_l_01 [{shapeType=Capsule, length=3.3162, radius=0.5, center=<0 0 -2>}]
 dyn_bellybow_string_l_02 [{shapeType=Capsule, length=4.1474, radius=0.5, center=<0.0001 0.5086 -1.8637>, rotation=<rot 0 0 -15.8404>}]
 dyn_bellybow_string_r_01 [{shapeType=Capsule, length=3.3162, radius=0.5, center=<0 0 -2>}]
 dyn_bellybow_string_r_02 [{shapeType=Capsule, length=4.5857, radius=0.5, center=<0.0001 -0.4963 -2.9182>, rotation=<rot 0.0001 0 14.3182>}]
 dyn_chestbow_base [{shapeType=Capsule, length=1, radius=1}]
 dyn_chestbow_loop_l [{shapeType=Capsule, length=3.5474, radius=2.7104, center=<0 0 -4>}]
 dyn_chestbow_loop_r [{shapeType=Capsule, length=3.5474, radius=2.7104, center=<0 0 -4>}]
 dyn_chestbow_string_l [{shapeType=Capsule, length=1, radius=1, center=<0 0 -1.5>}]
 dyn_chestbow_string_r [{shapeType=Capsule, length=1, radius=1, center=<0 0 -1.5>}]
 dyn_choker_l [{shapeType=Capsule, length=1.5, radius=0.5, center=<0 0 -1.5>}]
 dyn_choker_r [{shapeType=Capsule, length=1.5, radius=0.5, center=<0 0 -1.5>}]
 neck_02 [{shapeType=Capsule, length=1, radius=1}]
 pelvis [{shapeType=Capsule, length=5, radius=5}]
 spine_03 [{shapeType=Capsule, length=7.7864, radius=8.9465, center=<-0.33 2.7639 0>, rotation=<rot 90 0 -8.44>}]
 spine_05 [{shapeType=Box, extentX=7.5, extentY=4.2729, extentZ=12.121, center=<4.3403 5.046 0>, rotation=<rot 0 -29.9994 0>}]
}
constraints [
 {bone1Name=dyn_bellybow_loop_r, bone2Name=dyn_bellybow_base, swing1LimitDegrees=30, swing2LimitDegrees=45, twistLimitDegrees=10}
 {bone1Name=dyn_chestbow_loop_r, bone2Name=dyn_chestbow_base, swing1LimitDegrees=30, swing2LimitDegrees=45, twistLimitDegrees=10}
 {bone1Name=dyn_chestbow_loop_l, bone2Name=dyn_chestbow_base, swing1LimitDegrees=30, swing2LimitDegrees=45, twistLimitDegrees=10}
 {bone1Name=dyn_chestbow_string_r, bone2Name=dyn_chestbow_base, swing1LimitDegrees=45, swing2LimitDegrees=45, twistLimitDegrees=10}
 {bone1Name=dyn_chestbow_string_l, bone2Name=dyn_chestbow_base, swing1LimitDegrees=45, swing2LimitDegrees=45, twistLimitDegrees=10}
 {bone1Name=dyn_bellybow_loop_l, bone2Name=dyn_bellybow_base, swing1LimitDegrees=30, swing2LimitDegrees=45, twistLimitDegrees=10}
 {bone1Name=dyn_bellybow_string_l_01, bone2Name=dyn_bellybow_base, swing1LimitDegrees=45, swing2LimitDegrees=45, twistLimitDegrees=10}
 {bone1Name=dyn_bellybow_string_l_02, bone2Name=dyn_bellybow_string_l_01, swing1LimitDegrees=45, swing2LimitDegrees=45, twistLimitDegrees=10}
 {bone1Name=dyn_bellybow_string_r_01, bone2Name=dyn_bellybow_base, swing1LimitDegrees=45, swing2LimitDegrees=45, twistLimitDegrees=10}
 {bone1Name=dyn_bellybow_string_r_02, bone2Name=dyn_bellybow_string_r_01, swing1LimitDegrees=45, swing2LimitDegrees=45, twistLimitDegrees=10}
 {bone1Name=dyn_choker_l, bone2Name=spine_05, swing1LimitDegrees=45, swing2LimitDegrees=45, twistLimitDegrees=10}
 {bone1Name=dyn_choker_r, bone2Name=spine_05, swing1LimitDegrees=45, swing2LimitDegrees=45, twistLimitDegrees=10}
 {
  bone1Name=neck_02
  bone2Name=spine_05
  swing1LimitDegrees=45
  swing1Motion=Free
  swing2LimitDegrees=45
  swing2Motion=Free
  twistLimitDegrees=45
  twistMotion=Free
 }
 {
  bone1Name=dyn_chestbow_base
  bone2Name=spine_05
  swing1LimitDegrees=45
  swing1Motion=Free
  swing2LimitDegrees=45
  swing2Motion=Free
  twistLimitDegrees=45
  twistMotion=Free
 }
 {
  bone1Name=dyn_bellybow_base
  bone2Name=spine_03
  swing1LimitDegrees=45
  swing1Motion=Free
  swing2LimitDegrees=45
  swing2Motion=Free
  twistLimitDegrees=45
  twistMotion=Free
 }
 {
  bone1Name=spine_03
  bone2Name=pelvis
  swing1LimitDegrees=45
  swing1Motion=Free
  swing2LimitDegrees=45
  swing2Motion=Free
  twistLimitDegrees=45
  twistMotion=Free
 }
 {
  bone1Name=spine_05
  bone2Name=spine_03
  swing1LimitDegrees=45
  swing1Motion=Free
  swing2LimitDegrees=45
  swing2Motion=Free
  twistLimitDegrees=45
  twistMotion=Free
 }
]
