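# Animation Blueprint for the female medium Octoberfest body cosmetic that adds secondary motion to its skirt.
# Chained AnimDynamics nodes simulate skirt bones, then crouch-weighted LookAt and bone transforms pull them toward the thighs, feeding a blend switched by bRigidBodyEnabled.
# A pelvis-space RigidBody node uses a skirt physics asset, and a cached 'Rigid' pose receives direction-weighted corrections.
# Several other flap, constraint and skydiving-force chains have unconnected inputs.
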AnimBlueprint /BRCosmetics/Characters/Player/Female/Medium/Bodies/F_MED_Octoberfest/Meshes/F_MED_Octoberfest_AnimBP
target_skeleton=&/F_MED_Octoberfest_Skeleton
graphs [
 AnimGraph {nodes [
   n1 = Anim Dynamics - Bone: dyn_flap_fr_l_01(Alpha=1, ExternalForce="0,0,0") {
    angularDampingOverride=0.999
    bGravityOverrideInSimSpace=true
    bOverrideAngularDamping=true
    bUseGravityOverride=true
    alphaBoolBlend {blendOption=Linear}
    boundBone {boneName=dyn_flap_fr_l_01}
    componentAppliedLinearAccClamp=<2000 2000 2000>
    componentLinearAccScale {z=0.8}
    componentLinearVelScale {z=0.8}
    gravityOverride {z=-980}
    physicsBodyDefinitions [{
      sphereCollisionRadius=0
      boundBone {boneName=dyn_flap_fr_l_01}
      boxExtents=<15 15 8>
      constraintSetup {twistAxis=AxisZ, angularLimitsMax=<0 0 0>, angularLimitsMin=<-10 -30 0>, angularTarget=<0 0 0>}
      localJointOffset=<0 0 -5>
     }]
   }
   n2 = Anim Dynamics - Bone: dyn_flap_fr_l_02(ComponentPose=@n1, Alpha=1, ExternalForce="0,0,0") {
    angularDampingOverride=0.999
    bGravityOverrideInSimSpace=true
    bOverrideAngularDamping=true
    bUseGravityOverride=true
    alphaBoolBlend {blendOption=Linear}
    boundBone {boneName=dyn_flap_fr_l_02}
    componentAppliedLinearAccClamp=<2000 2000 2000>
    componentLinearAccScale {z=0.8}
    componentLinearVelScale {z=0.8}
    gravityOverride {z=-980}
    physicsBodyDefinitions [{
      sphereCollisionRadius=0
      boundBone {boneName=dyn_flap_fr_l_02}
      boxExtents=<15 15 8>
      constraintSetup {twistAxis=AxisZ, angularLimitsMax=<10 0 0>, angularLimitsMin=<-10 -30 0>, angularTarget=<0 0 0>}
      localJointOffset=<0 0 -5>
     }]
   }
   n3 = Anim Dynamics - Bone: dyn_flap_fr_r_01(ComponentPose=@n2, Alpha=1, ExternalForce="0,0,0") {
    angularDampingOverride=0.999
    bGravityOverrideInSimSpace=true
    bOverrideAngularDamping=true
    bUseGravityOverride=true
    alphaBoolBlend {blendOption=Linear}
    boundBone {boneName=dyn_flap_fr_r_01}
    componentAppliedLinearAccClamp=<2000 2000 2000>
    componentLinearAccScale {z=0.8}
    componentLinearVelScale {z=0.8}
    gravityOverride {z=-980}
    physicsBodyDefinitions [{
      sphereCollisionRadius=0
      boundBone {boneName=dyn_flap_fr_r_01}
      boxExtents=<15 15 8>
      constraintSetup {twistAxis=AxisZ, angularLimitsMax=<0 0 0>, angularLimitsMin=<-10 -30 0>, angularTarget=<0 0 0>}
      localJointOffset=<0 0 5>
     }]
   }
   Anim Dynamics - Bone: dyn_flap_fr_r_02(ComponentPose=@n3, Alpha=1, ExternalForce="0,0,0") {
    angularDampingOverride=0.999
    bGravityOverrideInSimSpace=true
    bOverrideAngularDamping=true
    bUseGravityOverride=true
    alphaBoolBlend {blendOption=Linear}
    boundBone {boneName=dyn_flap_fr_r_02}
    componentAppliedLinearAccClamp=<2000 2000 2000>
    componentLinearAccScale {z=0.8}
    componentLinearVelScale {z=0.8}
    gravityOverride {z=-980}
    physicsBodyDefinitions [{
      sphereCollisionRadius=0
      boundBone {boneName=dyn_flap_fr_r_02}
      boxExtents=<15 15 8>
      constraintSetup {twistAxis=AxisZ, angularLimitsMax=<10 0 0>, angularLimitsMin=<-10 -30 0>, angularTarget=<0 0 0>}
      localJointOffset=<0 0 5>
     }]
   }
   n4 = Anim Dynamics - Bone: dyn_skirt_fr_01(ComponentPose=Local To Component(LocalPose=Use cached pose 'initial Pose'()), Alpha=1, ExternalForce="0,0,0") {
    angularDampingOverride=0.9999
    angularSpringConstant=500
    bAngularSpring=true
    bGravityOverrideInSimSpace=true
    bOverrideAngularDamping=true
    bUseSphericalLimits=true
    simulationSpace=BoneRelative
    alphaBoolBlend {blendOption=Linear}
    boundBone {boneName=dyn_skirt_fr_01}
    componentAppliedLinearAccClamp=<3500 3500 3500>
    componentLinearAccScale=<1.2 1.2 1.2>
    componentLinearVelScale=<1.2 1.2 1.2>
    physicsBodyDefinitions [{
      sphereCollisionRadius=0
      boundBone {boneName=dyn_skirt_fr_01}
      boxExtents=<15 20 20>
      constraintSetup {angularLimitsMax=<1 5 15>, angularLimitsMin=<-1 -45 -15>, angularTarget=<0 0 0>}
      localJointOffset=<3 0 0>
     }]
    relativeSpaceBone {boneName=root}
   }
   n5 = Anim Dynamics - Bone: dyn_skirt_fr_02(ComponentPose=@n4, Alpha=1, ExternalForce="0,0,0") {
    angularDampingOverride=0.9999
    angularSpringConstant=500
    bAngularSpring=true
    bGravityOverrideInSimSpace=true
    bOverrideAngularDamping=true
    simulationSpace=BoneRelative
    alphaBoolBlend {blendOption=Linear}
    boundBone {boneName=dyn_skirt_fr_02}
    componentAppliedLinearAccClamp=<3500 3500 3500>
    componentLinearAccScale=<1.2 1.2 1.2>
    componentLinearVelScale=<1.2 1.2 1.2>
    physicsBodyDefinitions [{
      sphereCollisionRadius=0
      boundBone {boneName=dyn_skirt_fr_02}
      boxExtents=<15 20 20>
      constraintSetup {angularLimitsMax=<1 8 15>, angularLimitsMin=<-1 -45 -15>, angularTarget=<0 0 0>}
      localJointOffset=<3 0 0>
     }]
    relativeSpaceBone {boneName=root}
   }
   n6 = Anim Dynamics - Bone: dyn_skirt_fr_03(ComponentPose=@n5, Alpha=1, ExternalForce="0,0,0") {
    angularDampingOverride=0.9999
    angularSpringConstant=500
    bAngularSpring=true
    bGravityOverrideInSimSpace=true
    bOverrideAngularDamping=true
    simulationSpace=BoneRelative
    alphaBoolBlend {blendOption=Linear}
    boundBone {boneName=dyn_skirt_fr_03}
    componentAppliedLinearAccClamp=<3500 3500 3500>
    componentLinearAccScale=<1.2 1.2 1.2>
    componentLinearVelScale=<1.2 1.2 1.2>
    physicsBodyDefinitions [{
      sphereCollisionRadius=0
      boundBone {boneName=dyn_skirt_fr_03}
      boxExtents=<15 20 20>
      constraintSetup {angularLimitsMax=<1 8 15>, angularLimitsMin=<-1 -45 -15>, angularTarget=<0 0 0>}
      localJointOffset=<3 0 0>
     }]
    relativeSpaceBone {boneName=root}
   }
   n7 = Anim Dynamics - Bone: dyn_skirt_fr_l_01(ComponentPose=@n6, Alpha=1, ExternalForce="0,0,0") {
    angularDampingOverride=0.9999
    angularSpringConstant=500
    bAngularSpring=true
    bGravityOverrideInSimSpace=true
    bOverrideAngularDamping=true
    bUseSphericalLimits=true
    simulationSpace=BoneRelative
    alphaBoolBlend {blendOption=Linear}
    boundBone {boneName=dyn_skirt_fr_l_01}
    componentAppliedLinearAccClamp=<3500 3500 3500>
    componentLinearAccScale=<1.2 1.2 1.2>
    componentLinearVelScale=<1.2 1.2 1.2>
    physicsBodyDefinitions [{
      sphereCollisionRadius=0
      boundBone {boneName=dyn_skirt_fr_l_01}
      boxExtents=<15 20 20>
      constraintSetup {angularLimitsMax=<1 5 15>, angularLimitsMin=<-1 -45 -15>, angularTarget=<0 0 0>}
      localJointOffset=<3 0 0>
     }]
    relativeSpaceBone {boneName=root}
   }
   n8 = Anim Dynamics - Bone: dyn_skirt_fr_l_02(ComponentPose=@n7, Alpha=1, ExternalForce="0,0,0") {
    angularDampingOverride=0.9999
    angularSpringConstant=500
    bAngularSpring=true
    bGravityOverrideInSimSpace=true
    bOverrideAngularDamping=true
    simulationSpace=BoneRelative
    alphaBoolBlend {blendOption=Linear}
    boundBone {boneName=dyn_skirt_fr_l_02}
    componentAppliedLinearAccClamp=<3500 3500 3500>
    componentLinearAccScale=<1.2 1.2 1.2>
    componentLinearVelScale=<1.2 1.2 1.2>
    physicsBodyDefinitions [{
      sphereCollisionRadius=0
      boundBone {boneName=dyn_skirt_fr_l_02}
      boxExtents=<15 20 20>
      constraintSetup {angularLimitsMax=<1 8 15>, angularLimitsMin=<-1 -45 -15>, angularTarget=<0 0 0>}
      localJointOffset=<3 0 0>
     }]
    relativeSpaceBone {boneName=root}
   }
   n9 = Anim Dynamics - Bone: dyn_skirt_fr_l_03(ComponentPose=@n8, Alpha=1, ExternalForce="0,0,0") {
    angularDampingOverride=0.9999
    angularSpringConstant=500
    bAngularSpring=true
    bGravityOverrideInSimSpace=true
    bOverrideAngularDamping=true
    simulationSpace=BoneRelative
    alphaBoolBlend {blendOption=Linear}
    boundBone {boneName=dyn_skirt_fr_l_03}
    componentAppliedLinearAccClamp=<3500 3500 3500>
    componentLinearAccScale=<1.2 1.2 1.2>
    componentLinearVelScale=<1.2 1.2 1.2>
    physicsBodyDefinitions [{
      sphereCollisionRadius=0
      boundBone {boneName=dyn_skirt_fr_l_03}
      boxExtents=<15 20 20>
      constraintSetup {angularLimitsMax=<1 8 15>, angularLimitsMin=<-1 -45 -15>, angularTarget=<0 0 0>}
      localJointOffset=<3 0 0>
     }]
    relativeSpaceBone {boneName=root}
   }
   n10 = Anim Dynamics - Bone: dyn_skirt_fr_r_01(ComponentPose=@n9, Alpha=1, ExternalForce="0,0,0") {
    angularDampingOverride=0.9999
    angularSpringConstant=500
    bAngularSpring=true
    bGravityOverrideInSimSpace=true
    bOverrideAngularDamping=true
    bUseSphericalLimits=true
    simulationSpace=BoneRelative
    alphaBoolBlend {blendOption=Linear}
    boundBone {boneName=dyn_skirt_fr_r_01}
    componentAppliedLinearAccClamp=<3500 3500 3500>
    componentLinearAccScale=<1.2 1.2 1.2>
    componentLinearVelScale=<1.2 1.2 1.2>
    physicsBodyDefinitions [{
      sphereCollisionRadius=0
      boundBone {boneName=dyn_skirt_fr_r_01}
      boxExtents=<15 20 20>
      constraintSetup {angularLimitsMax=<1 5 15>, angularLimitsMin=<-1 -45 -15>, angularTarget=<0 0 0>}
      localJointOffset=<3 0 0>
     }]
    relativeSpaceBone {boneName=root}
   }
   n11 = Anim Dynamics - Bone: dyn_skirt_fr_r_02(ComponentPose=@n10, Alpha=1, ExternalForce="0,0,0") {
    angularDampingOverride=0.9999
    angularSpringConstant=500
    bAngularSpring=true
    bGravityOverrideInSimSpace=true
    bOverrideAngularDamping=true
    simulationSpace=BoneRelative
    alphaBoolBlend {blendOption=Linear}
    boundBone {boneName=dyn_skirt_fr_r_02}
    componentAppliedLinearAccClamp=<3500 3500 3500>
    componentLinearAccScale=<1.2 1.2 1.2>
    componentLinearVelScale=<1.2 1.2 1.2>
    physicsBodyDefinitions [{
      sphereCollisionRadius=0
      boundBone {boneName=dyn_skirt_fr_r_02}
      boxExtents=<15 20 20>
      constraintSetup {angularLimitsMax=<1 8 15>, angularLimitsMin=<-1 -45 -15>, angularTarget=<0 0 0>}
      localJointOffset=<3 0 0>
     }]
    relativeSpaceBone {boneName=root}
   }
   n12 = Anim Dynamics - Bone: dyn_skirt_fr_r_03(ComponentPose=@n11, Alpha=1, ExternalForce="0,0,0") {
    angularDampingOverride=0.9999
    angularSpringConstant=500
    bAngularSpring=true
    bGravityOverrideInSimSpace=true
    bOverrideAngularDamping=true
    simulationSpace=BoneRelative
    alphaBoolBlend {blendOption=Linear}
    boundBone {boneName=dyn_skirt_fr_r_03}
    componentAppliedLinearAccClamp=<3500 3500 3500>
    componentLinearAccScale=<1.2 1.2 1.2>
    componentLinearVelScale=<1.2 1.2 1.2>
    physicsBodyDefinitions [{
      sphereCollisionRadius=0
      boundBone {boneName=dyn_skirt_fr_r_03}
      boxExtents=<15 20 20>
      constraintSetup {angularLimitsMax=<1 8 15>, angularLimitsMin=<-1 -45 -15>, angularTarget=<0 0 0>}
      localJointOffset=<3 0 0>
     }]
    relativeSpaceBone {boneName=root}
   }
   n13 = Anim Dynamics - Bone: dyn_skirt_out_l_01(ComponentPose=@n12, Alpha=1, ExternalForce="0,0,0") {
    angularDampingOverride=0.9999
    angularSpringConstant=500
    bAngularSpring=true
    bGravityOverrideInSimSpace=true
    bOverrideAngularDamping=true
    bUseSphericalLimits=true
    simulationSpace=BoneRelative
    alphaBoolBlend {blendOption=Linear}
    boundBone {boneName=dyn_skirt_out_l_01}
    componentAppliedLinearAccClamp=<3500 3500 3500>
    componentLinearAccScale=<1.2 1.2 1.2>
    componentLinearVelScale=<1.2 1.2 1.2>
    physicsBodyDefinitions [{
      sphereCollisionRadius=0
      boundBone {boneName=dyn_skirt_out_l_01}
      boxExtents=<15 20 20>
      constraintSetup {angularLimitsMax=<1 5 15>, angularLimitsMin=<-1 -45 -15>, angularTarget=<0 0 0>}
      localJointOffset=<3 0 0>
     }]
    relativeSpaceBone {boneName=root}
   }
   n14 = Anim Dynamics - Bone: dyn_skirt_out_l_02(ComponentPose=@n13, Alpha=1, ExternalForce="0,0,0") {
    angularDampingOverride=0.9999
    angularSpringConstant=500
    bAngularSpring=true
    bGravityOverrideInSimSpace=true
    bOverrideAngularDamping=true
    simulationSpace=BoneRelative
    alphaBoolBlend {blendOption=Linear}
    boundBone {boneName=dyn_skirt_out_l_02}
    componentAppliedLinearAccClamp=<3500 3500 3500>
    componentLinearAccScale=<1.2 1.2 1.2>
    componentLinearVelScale=<1.2 1.2 1.2>
    physicsBodyDefinitions [{
      sphereCollisionRadius=0
      boundBone {boneName=dyn_skirt_out_l_02}
      boxExtents=<15 20 20>
      constraintSetup {angularLimitsMax=<1 8 15>, angularLimitsMin=<-1 -45 -15>, angularTarget=<0 0 0>}
      localJointOffset=<3 0 0>
     }]
    relativeSpaceBone {boneName=root}
   }
   n15 = Anim Dynamics - Bone: dyn_skirt_out_l_03(ComponentPose=@n14, Alpha=1, ExternalForce="0,0,0") {
    angularDampingOverride=0.9999
    angularSpringConstant=500
    bAngularSpring=true
    bGravityOverrideInSimSpace=true
    bOverrideAngularDamping=true
    simulationSpace=BoneRelative
    alphaBoolBlend {blendOption=Linear}
    boundBone {boneName=dyn_skirt_out_l_03}
    componentAppliedLinearAccClamp=<3500 3500 3500>
    componentLinearAccScale=<1.2 1.2 1.2>
    componentLinearVelScale=<1.2 1.2 1.2>
    physicsBodyDefinitions [{
      sphereCollisionRadius=0
      boundBone {boneName=dyn_skirt_out_l_03}
      boxExtents=<15 20 20>
      constraintSetup {angularLimitsMax=<1 8 15>, angularLimitsMin=<-1 -45 -15>, angularTarget=<0 0 0>}
      localJointOffset=<3 0 0>
     }]
    relativeSpaceBone {boneName=root}
   }
   n16 = Anim Dynamics - Bone: dyn_skirt_out_r_01(ComponentPose=@n15, Alpha=1, ExternalForce="0,0,0") {
    angularDampingOverride=0.9999
    angularSpringConstant=500
    bAngularSpring=true
    bGravityOverrideInSimSpace=true
    bOverrideAngularDamping=true
    bUseSphericalLimits=true
    simulationSpace=BoneRelative
    alphaBoolBlend {blendOption=Linear}
    boundBone {boneName=dyn_skirt_out_r_01}
    componentAppliedLinearAccClamp=<3500 3500 3500>
    componentLinearAccScale=<1.2 1.2 1.2>
    componentLinearVelScale=<1.2 1.2 1.2>
    physicsBodyDefinitions [{
      sphereCollisionRadius=0
      boundBone {boneName=dyn_skirt_out_r_01}
      boxExtents=<15 20 20>
      constraintSetup {angularLimitsMax=<1 5 15>, angularLimitsMin=<-1 -45 -15>, angularTarget=<0 0 0>}
      localJointOffset=<3 0 0>
     }]
    relativeSpaceBone {boneName=root}
   }
   n17 = Anim Dynamics - Bone: dyn_skirt_out_r_02(ComponentPose=@n16, Alpha=1, ExternalForce="0,0,0") {
    angularDampingOverride=0.9999
    angularSpringConstant=500
    bAngularSpring=true
    bGravityOverrideInSimSpace=true
    bOverrideAngularDamping=true
    simulationSpace=BoneRelative
    alphaBoolBlend {blendOption=Linear}
    boundBone {boneName=dyn_skirt_out_r_02}
    componentAppliedLinearAccClamp=<3500 3500 3500>
    componentLinearAccScale=<1.2 1.2 1.2>
    componentLinearVelScale=<1.2 1.2 1.2>
    physicsBodyDefinitions [{
      sphereCollisionRadius=0
      boundBone {boneName=dyn_skirt_out_r_02}
      boxExtents=<15 20 20>
      constraintSetup {angularLimitsMax=<1 8 15>, angularLimitsMin=<-1 -45 -15>, angularTarget=<0 0 0>}
      localJointOffset=<3 0 0>
     }]
    relativeSpaceBone {boneName=root}
   }
   n18 = Anim Dynamics - Bone: dyn_skirt_out_r_03(ComponentPose=@n17, Alpha=1, ExternalForce="0,0,0") {
    angularDampingOverride=0.9999
    angularSpringConstant=500
    bAngularSpring=true
    bGravityOverrideInSimSpace=true
    bOverrideAngularDamping=true
    simulationSpace=BoneRelative
    alphaBoolBlend {blendOption=Linear}
    boundBone {boneName=dyn_skirt_out_r_03}
    componentAppliedLinearAccClamp=<3500 3500 3500>
    componentLinearAccScale=<1.2 1.2 1.2>
    componentLinearVelScale=<1.2 1.2 1.2>
    physicsBodyDefinitions [{
      sphereCollisionRadius=0
      boundBone {boneName=dyn_skirt_out_r_03}
      boxExtents=<15 20 20>
      constraintSetup {angularLimitsMax=<1 8 15>, angularLimitsMin=<-1 -45 -15>, angularTarget=<0 0 0>}
      localJointOffset=<3 0 0>
     }]
    relativeSpaceBone {boneName=root}
   }
   n19 = Anim Dynamics - Bone: dyn_skirt_bk_l_01(ComponentPose=@n18, Alpha=1, ExternalForce="0,0,0") {
    angularDampingOverride=0.9999
    angularSpringConstant=500
    bAngularSpring=true
    bGravityOverrideInSimSpace=true
    bOverrideAngularDamping=true
    bUseSphericalLimits=true
    simulationSpace=BoneRelative
    alphaBoolBlend {blendOption=Linear}
    boundBone {boneName=dyn_skirt_bk_l_01}
    componentAppliedLinearAccClamp=<3500 3500 3500>
    componentLinearAccScale=<1.2 1.2 1.2>
    componentLinearVelScale=<1.2 1.2 1.2>
    physicsBodyDefinitions [{
      sphereCollisionRadius=0
      boundBone {boneName=dyn_skirt_bk_l_01}
      boxExtents=<15 20 20>
      constraintSetup {angularLimitsMax=<1 5 15>, angularLimitsMin=<-1 -45 -15>, angularTarget=<0 0 0>}
      localJointOffset=<3 0 0>
     }]
    relativeSpaceBone {boneName=root}
   }
   n20 = Anim Dynamics - Bone: dyn_skirt_bk_l_02(ComponentPose=@n19, Alpha=1, ExternalForce="0,0,0") {
    angularDampingOverride=0.9999
    angularSpringConstant=500
    bAngularSpring=true
    bGravityOverrideInSimSpace=true
    bOverrideAngularDamping=true
    simulationSpace=BoneRelative
    alphaBoolBlend {blendOption=Linear}
    boundBone {boneName=dyn_skirt_bk_l_02}
    componentAppliedLinearAccClamp=<3500 3500 3500>
    componentLinearAccScale=<1.2 1.2 1.2>
    componentLinearVelScale=<1.2 1.2 1.2>
    physicsBodyDefinitions [{
      sphereCollisionRadius=0
      boundBone {boneName=dyn_skirt_bk_l_02}
      boxExtents=<15 20 20>
      constraintSetup {angularLimitsMax=<1 8 15>, angularLimitsMin=<-1 -45 -15>, angularTarget=<0 0 0>}
      localJointOffset=<3 0 0>
     }]
    relativeSpaceBone {boneName=root}
   }
   n21 = Anim Dynamics - Bone: dyn_skirt_bk_l_03(ComponentPose=@n20, Alpha=1, ExternalForce="0,0,0") {
    angularDampingOverride=0.9999
    angularSpringConstant=500
    bAngularSpring=true
    bGravityOverrideInSimSpace=true
    bOverrideAngularDamping=true
    simulationSpace=BoneRelative
    alphaBoolBlend {blendOption=Linear}
    boundBone {boneName=dyn_skirt_bk_l_03}
    componentAppliedLinearAccClamp=<3500 3500 3500>
    componentLinearAccScale=<1.2 1.2 1.2>
    componentLinearVelScale=<1.2 1.2 1.2>
    physicsBodyDefinitions [{
      sphereCollisionRadius=0
      boundBone {boneName=dyn_skirt_bk_l_03}
      boxExtents=<15 20 20>
      constraintSetup {angularLimitsMax=<1 8 15>, angularLimitsMin=<-1 -45 -15>, angularTarget=<0 0 0>}
      localJointOffset=<3 0 0>
     }]
    relativeSpaceBone {boneName=root}
   }
   n22 = Anim Dynamics - Bone: dyn_skirt_bk_r_01(ComponentPose=@n21, Alpha=1, ExternalForce="0,0,0") {
    angularDampingOverride=0.9999
    angularSpringConstant=500
    bAngularSpring=true
    bGravityOverrideInSimSpace=true
    bOverrideAngularDamping=true
    bUseSphericalLimits=true
    simulationSpace=BoneRelative
    alphaBoolBlend {blendOption=Linear}
    boundBone {boneName=dyn_skirt_bk_r_01}
    componentAppliedLinearAccClamp=<3500 3500 3500>
    componentLinearAccScale=<1.2 1.2 1.2>
    componentLinearVelScale=<1.2 1.2 1.2>
    physicsBodyDefinitions [{
      sphereCollisionRadius=0
      boundBone {boneName=dyn_skirt_bk_r_01}
      boxExtents=<15 20 20>
      constraintSetup {angularLimitsMax=<1 5 15>, angularLimitsMin=<-1 -45 -15>, angularTarget=<0 0 0>}
      localJointOffset=<3 0 0>
     }]
    relativeSpaceBone {boneName=root}
   }
   n23 = Anim Dynamics - Bone: dyn_skirt_bk_r_02(ComponentPose=@n22, Alpha=1, ExternalForce="0,0,0") {
    angularDampingOverride=0.9999
    angularSpringConstant=500
    bAngularSpring=true
    bGravityOverrideInSimSpace=true
    bOverrideAngularDamping=true
    simulationSpace=BoneRelative
    alphaBoolBlend {blendOption=Linear}
    boundBone {boneName=dyn_skirt_bk_r_02}
    componentAppliedLinearAccClamp=<3500 3500 3500>
    componentLinearAccScale=<1.2 1.2 1.2>
    componentLinearVelScale=<1.2 1.2 1.2>
    physicsBodyDefinitions [{
      sphereCollisionRadius=0
      boundBone {boneName=dyn_skirt_bk_r_02}
      boxExtents=<15 20 20>
      constraintSetup {angularLimitsMax=<1 8 15>, angularLimitsMin=<-1 -45 -15>, angularTarget=<0 0 0>}
      localJointOffset=<3 0 0>
     }]
    relativeSpaceBone {boneName=root}
   }
   n24 = Anim Dynamics - Bone: dyn_skirt_bk_r_03(ComponentPose=@n23, Alpha=1, ExternalForce="0,0,0") {
    angularDampingOverride=0.9999
    angularSpringConstant=500
    bAngularSpring=true
    bGravityOverrideInSimSpace=true
    bOverrideAngularDamping=true
    simulationSpace=BoneRelative
    alphaBoolBlend {blendOption=Linear}
    boundBone {boneName=dyn_skirt_bk_r_03}
    componentAppliedLinearAccClamp=<3500 3500 3500>
    componentLinearAccScale=<1.2 1.2 1.2>
    componentLinearVelScale=<1.2 1.2 1.2>
    physicsBodyDefinitions [{
      sphereCollisionRadius=0
      boundBone {boneName=dyn_skirt_bk_r_03}
      boxExtents=<15 20 20>
      constraintSetup {angularLimitsMax=<1 8 15>, angularLimitsMin=<-1 -45 -15>, angularTarget=<0 0 0>}
      localJointOffset=<3 0 0>
     }]
    relativeSpaceBone {boneName=root}
   }
   n25 = Get Crouch 0ff() -> (Crouch 0ff:double)
   n26 = Look At - Bone: dyn_skirt_fr_l_MSTR(ComponentPose=@n24, Alpha=@n25) {
    alphaBoolBlend {blendInTime=0.7, blendOption=Linear, blendOutTime=0.5}
    alphaScaleBiasClamp {bClampResult=true, bInterpResult=true, clampMax=0.8, interpSpeedDecreasing=5, interpSpeedIncreasing=7}
    boneToModify {boneName=dyn_skirt_fr_l_MSTR}
    lookAtLocation=<-52.1775 29.2072 -5.9906>
    lookAtTarget {boneReference {boneName=thigh_l}}
    lookAt_Axis {axis=<1 0>}
   }
   n27 = Look At - Bone: dyn_skirt_fr_r_MSTR(ComponentPose=@n26, Alpha=@n25) {
    alphaBoolBlend {blendInTime=0.7, blendOption=Linear, blendOutTime=0.5}
    alphaScaleBiasClamp {bClampResult=true, bInterpResult=true, clampMax=0.8, interpSpeedDecreasing=5, interpSpeedIncreasing=7}
    boneToModify {boneName=dyn_skirt_fr_r_MSTR}
    lookAtLocation=<50.1405 -55.4538 2.8637>
    lookAtTarget {boneReference {boneName=thigh_r}}
    lookAt_Axis {axis=<1 0>}
   }
   n28 = Look At - Bone: dyn_skirt_out_l_MSTR(ComponentPose=@n27, Alpha=@n25) {
    alphaBoolBlend {blendInTime=0.7, blendOption=Linear, blendOutTime=0.5}
    alphaScaleBiasClamp {bClampResult=true, bInterpResult=true, clampMax=0.8, interpSpeedDecreasing=5, interpSpeedIncreasing=7}
    boneToModify {boneName=dyn_skirt_out_l_MSTR}
    lookAtLocation=<-43.0073 2.4629 -32.3146>
    lookAtTarget {boneReference {boneName=thigh_l}}
    lookAt_Axis {axis=<1 0>}
   }
   n29 = Look At - Bone: dyn_skirt_out_r_MSTR(ComponentPose=@n28, Alpha=@n25) {
    alphaBoolBlend {blendInTime=0.7, blendOption=Linear, blendOutTime=0.5}
    alphaScaleBiasClamp {bClampResult=true, bInterpResult=true, clampMax=0.8, interpSpeedDecreasing=5, interpSpeedIncreasing=7}
    boneToModify {boneName=dyn_skirt_out_r_MSTR}
    lookAtLocation=<45.7432 -1.4968 39.1723>
    lookAtTarget {boneReference {boneName=thigh_r}}
    lookAt_Axis {axis=<1 0>}
   }
   n30 = Look At - Bone: dyn_skirt_bk_l_01(ComponentPose=@n29, Alpha=@n25) {
    alphaBoolBlend {blendInTime=0.7, blendOption=Linear, blendOutTime=0.5}
    alphaScaleBiasClamp {bClampResult=true, bInterpResult=true, clampMax=0.8, interpSpeedDecreasing=5, interpSpeedIncreasing=7}
    boneToModify {boneName=dyn_skirt_bk_l_01}
    lookAtLocation=<-51.6517 -65.6618 -3.4972>
    lookAtTarget {boneReference {boneName=thigh_l}}
    lookAt_Axis {axis=<1 0>}
   }
   n31 = Look At - Bone: dyn_skirt_bk_r_01(ComponentPose=@n30, Alpha=@n25) {
    alphaBoolBlend {blendInTime=0.7, blendOption=Linear, blendOutTime=0.5}
    alphaScaleBiasClamp {bClampResult=true, bInterpResult=true, clampMax=0.8, interpSpeedDecreasing=5, interpSpeedIncreasing=7}
    boneToModify {boneName=dyn_skirt_bk_r_01}
    lookAtLocation=<45.3231 48.6446 3.6812>
    lookAtTarget {boneReference {boneName=thigh_r}}
    lookAt_Axis {axis=<1 0>}
   }
   n32 = Get crouch alpha() -> (crouch alpha:double)
   n33 = "Transform (Modify) Bone - Bone: dyn_skirt_fr_l_01"!(ComponentPose=@n31, Alpha=@n32, Rotation="5.000,0.000,0") -> (Pose) {
    rotationMode=BMM_Replace
    rotationSpace=BCS_BoneSpace
    translationMode=BMM_Replace
    translationSpace=BCS_BoneSpace
    alphaBoolBlend {blendOption=Linear}
    alphaScaleBiasClamp {bInterpResult=true, interpSpeedDecreasing=5, interpSpeedIncreasing=7}
    boneToModify {boneName=dyn_skirt_fr_l_01}
    rotation {pitch=5}
   }
   n34 = "Transform (Modify) Bone - Bone: dyn_skirt_out_l_01"!(ComponentPose=@n33, Alpha=@n32, Rotation="33.400,18.953,-22.719") -> (Pose) {
    rotationMode=BMM_Replace
    rotationSpace=BCS_BoneSpace
    translationMode=BMM_Replace
    translationSpace=BCS_BoneSpace
    alphaBoolBlend {blendOption=Linear}
    alphaScaleBiasClamp {bInterpResult=true, interpSpeedDecreasing=5, interpSpeedIncreasing=7}
    boneToModify {boneName=dyn_skirt_out_l_01}
    rotation=<rot 33.4 18.953 -22.719>
   }
   n35 = "Transform (Modify) Bone - Bone: dyn_skirt_bk_l_01"!(ComponentPose=@n34, Alpha=@n32, Rotation="45.368,-16.043,23.553") -> (Pose) {
    rotationMode=BMM_Replace
    rotationSpace=BCS_BoneSpace
    translationMode=BMM_Replace
    translationSpace=BCS_BoneSpace
    alphaBoolBlend {blendOption=Linear}
    alphaScaleBiasClamp {bInterpResult=true, interpSpeedDecreasing=5, interpSpeedIncreasing=7}
    boneToModify {boneName=dyn_skirt_bk_l_01}
    rotation=<rot 45.368 -16.043 23.553>
   }
   Blend Poses by bool(BlendPose_1=Component To Local(ComponentPose=@n35), bActiveValue=Get bRigidBodyEnabled()) {blendType=Linear}
   n36 = "Constraint - dyn_flap_fr_l_01 (R)"!(ComponentPose, Alpha=1) -> (Pose) {
    alphaBoolBlend {blendOption=Linear}
    boneToModify {boneName=dyn_flap_fr_l_01}
    constraintSetup [{transformType=Rotation, targetBone {boneName=dyn_skirt_fr_01}}]
    constraintWeights [1]
   }
   "Constraint - dyn_flap_fr_r_01 (R)"!(ComponentPose=@n36, Alpha=1) -> (Pose) {
    alphaBoolBlend {blendOption=Linear}
    boneToModify {boneName=dyn_flap_fr_r_01}
    constraintSetup [{transformType=Rotation, targetBone {boneName=dyn_skirt_fr_01}}]
    constraintWeights [1]
   }
   n37 = "Constraint - dyn_skirt_fr_l_01 (R)"!(ComponentPose, Alpha=1) -> (Pose) {
    alphaBoolBlend {blendOption=Linear}
    boneToModify {boneName=dyn_skirt_fr_l_01}
    constraintSetup [{transformType=Rotation, targetBone {boneName=thigh_l}}]
    constraintWeights [1]
   }
   n38 = "Constraint - dyn_skirt_fr_r_01 (R)"!(ComponentPose=@n37, Alpha=1) -> (Pose) {
    alphaBoolBlend {blendOption=Linear}
    boneToModify {boneName=dyn_skirt_fr_r_01}
    constraintSetup [{transformType=Rotation, targetBone {boneName=thigh_r}}]
    constraintWeights [1]
   }
   n39 = "Constraint - dyn_skirt_out_l_01 (R)"!(ComponentPose=@n38, Alpha=1) -> (Pose) {
    alphaBoolBlend {blendOption=Linear}
    boneToModify {boneName=dyn_skirt_out_l_01}
    constraintSetup [{transformType=Rotation, targetBone {boneName=thigh_l}}]
    constraintWeights [1]
   }
   n40 = "Constraint - dyn_skirt_out_r_01 (R)"!(ComponentPose=@n39, Alpha=1) -> (Pose) {
    alphaBoolBlend {blendOption=Linear}
    boneToModify {boneName=dyn_skirt_out_r_01}
    constraintSetup [{transformType=Rotation, targetBone {boneName=thigh_r}}]
    constraintWeights [1]
   }
   n41 = SelectVector(A=Get skydiving vector force() -> (skydiving vector force:Vector), B="0, 0, 0", bPickA=Get bIsSkyDiving())
   n42 = Anim Dynamics - Bone: dyn_skirt_fr_l_01(ComponentPose=@n40, Alpha=1, ExternalForce=@n41) {
    angularDampingOverride=0.98
    bGravityOverrideInSimSpace=true
    bOverrideAngularDamping=true
    bUseGravityOverride=true
    numSolverIterationsPostUpdate=2
    numSolverIterationsPreUpdate=8
    alphaBoolBlend {blendOption=Linear}
    boundBone {boneName=dyn_skirt_fr_l_01}
    gravityOverride {z=-980}
    physicsBodyDefinitions [{
      sphereCollisionRadius=0
      boundBone {boneName=dyn_skirt_fr_l_01}
      boxExtents=<20 20 10>
      constraintSetup {twistAxis=AxisZ, angularLimitsMax=<0 0 2>, angularLimitsMin=<0 -30 -2>, angularTarget=<0 0 0>}
      localJointOffset=<0 0 -3>
     }]
   }
   n43 = Anim Dynamics - Bone: dyn_skirt_fr_l_02(ComponentPose=@n42, Alpha=1, ExternalForce=@n41) {
    angularDampingOverride=0.98
    bGravityOverrideInSimSpace=true
    bOverrideAngularDamping=true
    bUseGravityOverride=true
    alphaBoolBlend {blendOption=Linear}
    boundBone {boneName=dyn_skirt_fr_l_02}
    gravityOverride {z=-980}
    physicsBodyDefinitions [{
      sphereCollisionRadius=0
      boundBone {boneName=dyn_skirt_fr_l_02}
      boxExtents=<20 20 10>
      constraintSetup {twistAxis=AxisZ, angularLimitsMax=<10 1 2>, angularLimitsMin=<-10 -30 -2>, angularTarget=<0 0 0>}
      localJointOffset=<0 0 -3>
     }]
   }
   n44 = Anim Dynamics - Bone: dyn_skirt_fr_l_03(ComponentPose=@n43, Alpha=1, ExternalForce=@n41) {
    angularDampingOverride=0.98
    bGravityOverrideInSimSpace=true
    bOverrideAngularDamping=true
    bUseGravityOverride=true
    alphaBoolBlend {blendOption=Linear}
    boundBone {boneName=dyn_skirt_fr_l_03}
    gravityOverride {z=-980}
    physicsBodyDefinitions [{
      sphereCollisionRadius=0
      boundBone {boneName=dyn_skirt_fr_l_03}
      boxExtents=<20 20 10>
      constraintSetup {twistAxis=AxisZ, angularLimitsMax=<10 0 2>, angularLimitsMin=<-10 -30 -2>, angularTarget=<0 0 0>}
      localJointOffset=<0 0 -3>
     }]
   }
   n45 = Anim Dynamics - Bone: dyn_skirt_fr_l_04(ComponentPose=@n44, Alpha=1, ExternalForce=@n41) {
    angularDampingOverride=0.98
    bGravityOverrideInSimSpace=true
    bOverrideAngularDamping=true
    bUseGravityOverride=true
    alphaBoolBlend {blendOption=Linear}
    boundBone {boneName=dyn_skirt_fr_l_04}
    gravityOverride {z=-980}
    physicsBodyDefinitions [{
      sphereCollisionRadius=0
      boundBone {boneName=dyn_skirt_fr_l_04}
      boxExtents=<20 20 10>
      constraintSetup {twistAxis=AxisZ, angularLimitsMax=<0 0 0>, angularLimitsMin=<0 -30 0>, angularTarget=<0 0 0>}
      localJointOffset=<0 0 -3>
     }]
   }
   n46 = Anim Dynamics - Bone: dyn_skirt_fr_r_01(ComponentPose=@n45, Alpha=1, ExternalForce=@n41) {
    angularDampingOverride=0.98
    bGravityOverrideInSimSpace=true
    bOverrideAngularDamping=true
    bUseGravityOverride=true
    alphaBoolBlend {blendOption=Linear}
    boundBone {boneName=dyn_skirt_fr_r_01}
    gravityOverride {z=-980}
    physicsBodyDefinitions [{
      sphereCollisionRadius=0
      boundBone {boneName=dyn_skirt_fr_r_01}
      boxExtents=<20 20 10>
      constraintSetup {twistAxis=AxisZ, angularLimitsMax=<0 0 2>, angularLimitsMin=<0 -30 -2>, angularTarget=<0 0 0>}
      localJointOffset=<0 0 3>
     }]
   }
   n47 = Anim Dynamics - Bone: dyn_skirt_fr_r_02(ComponentPose=@n46, Alpha=1, ExternalForce=@n41) {
    angularDampingOverride=0.98
    bGravityOverrideInSimSpace=true
    bOverrideAngularDamping=true
    bUseGravityOverride=true
    alphaBoolBlend {blendOption=Linear}
    boundBone {boneName=dyn_skirt_fr_r_02}
    gravityOverride {z=-980}
    physicsBodyDefinitions [{
      sphereCollisionRadius=0
      boundBone {boneName=dyn_skirt_fr_r_02}
      boxExtents=<20 20 10>
      constraintSetup {twistAxis=AxisZ, angularLimitsMax=<10 1 2>, angularLimitsMin=<-10 -30 -2>, angularTarget=<0 0 0>}
      localJointOffset=<0 0 3>
     }]
   }
   n48 = Anim Dynamics - Bone: dyn_skirt_fr_r_03(ComponentPose=@n47, Alpha=1, ExternalForce=@n41) {
    angularDampingOverride=0.98
    bGravityOverrideInSimSpace=true
    bOverrideAngularDamping=true
    bUseGravityOverride=true
    alphaBoolBlend {blendOption=Linear}
    boundBone {boneName=dyn_skirt_fr_r_03}
    gravityOverride {z=-980}
    physicsBodyDefinitions [{
      sphereCollisionRadius=0
      boundBone {boneName=dyn_skirt_fr_r_03}
      boxExtents=<20 20 10>
      constraintSetup {twistAxis=AxisZ, angularLimitsMax=<10 3 2>, angularLimitsMin=<-10 -30 -2>, angularTarget=<0 0 0>}
      localJointOffset=<0 0 3>
     }]
   }
   n49 = Anim Dynamics - Bone: dyn_skirt_fr_r_04(ComponentPose=@n48, Alpha=1, ExternalForce=@n41) {
    angularDampingOverride=0.98
    bGravityOverrideInSimSpace=true
    bOverrideAngularDamping=true
    bUseGravityOverride=true
    alphaBoolBlend {blendOption=Linear}
    boundBone {boneName=dyn_skirt_fr_r_04}
    gravityOverride {z=-980}
    physicsBodyDefinitions [{
      sphereCollisionRadius=0
      boundBone {boneName=dyn_skirt_fr_r_04}
      boxExtents=<20 20 10>
      constraintSetup {twistAxis=AxisZ, angularLimitsMax=<0 0 0>, angularLimitsMin=<0 -30 0>, angularTarget=<0 0 0>}
      localJointOffset=<0 0 3>
     }]
   }
   n50 = "Constraint - dyn_skirt_fr_01 (PP)"!(ComponentPose=@n49, Alpha=1, ConstraintWeights_0=1, ConstraintWeights_1:float=1) -> (Pose) {
    alphaBoolBlend {blendOption=Linear}
    boneToModify {boneName=dyn_skirt_fr_01}
    constraintSetup [{targetBone {boneName=dyn_skirt_fr_l_01}}, {targetBone {boneName=dyn_skirt_fr_r_01}}]
    constraintWeights [1,1]
   }
   n51 = "Constraint - dyn_skirt_fr_02 (RR)"!(ComponentPose=@n50, Alpha=1, ConstraintWeights_0=1, ConstraintWeights_1:float=1) -> (Pose) {
    alphaBoolBlend {blendOption=Linear}
    boneToModify {boneName=dyn_skirt_fr_02}
    constraintSetup [{transformType=Rotation, targetBone {boneName=dyn_skirt_fr_l_02}}, {transformType=Rotation, targetBone {boneName=dyn_skirt_fr_r_02}}]
    constraintWeights [1,1]
   }
   n52 = "Constraint - dyn_skirt_fr_03 (RR)"!(ComponentPose=@n51, Alpha=1, ConstraintWeights_0=1, ConstraintWeights_1:float=1) -> (Pose) {
    alphaBoolBlend {blendOption=Linear}
    boneToModify {boneName=dyn_skirt_fr_03}
    constraintSetup [{transformType=Rotation, targetBone {boneName=dyn_skirt_fr_l_03}}, {transformType=Rotation, targetBone {boneName=dyn_skirt_fr_r_03}}]
    constraintWeights [1,1]
   }
   "Constraint - dyn_skirt_fr_04 (RR)"!(ComponentPose=@n52, Alpha=1, ConstraintWeights_0=1, ConstraintWeights_1:float=1) -> (Pose) {
    alphaBoolBlend {blendOption=Linear}
    boneToModify {boneName=dyn_skirt_fr_04}
    constraintSetup [{transformType=Rotation, targetBone {boneName=dyn_skirt_fr_l_04}}, {transformType=Rotation, targetBone {boneName=dyn_skirt_fr_r_04}}]
    constraintWeights [1,1]
   }
   n53 = "Transform (Modify) Bone - Bone: dyn_skirt_fr_l_01"!(ComponentPose, Alpha=1, Translation:Vector="-1.948,-0.952,-1.832", Rotation="11.954,-10.111,-2", Scale:Vector="1,1,1") -> (Pose) {
    rotationMode=BMM_Additive
    rotationSpace=BCS_BoneSpace
    translationMode=BMM_Additive
    translationSpace=BCS_BoneSpace
    alphaBoolBlend {blendOption=Linear}
    boneToModify {boneName=dyn_skirt_fr_l_01}
    rotation=<rot 11.954 -10.111 -2>
    translation=<-1.948 -0.952 -1.832>
   }
   n54 = "Transform (Modify) Bone - Bone: dyn_skirt_fr_l_02"(ComponentPose=@n53, Alpha=1, Translation="-0.846,-0.208,-0.836", Rotation="-10.000,-10.000,-0.000", Scale="1,1,1") {
    rotationMode=BMM_Additive
    rotationSpace=BCS_BoneSpace
    translationMode=BMM_Additive
    translationSpace=BCS_BoneSpace
    alphaBoolBlend {blendOption=Linear}
    boneToModify {boneName=dyn_skirt_fr_l_02}
    rotation=<rot -10 -10.0002 0>
    translation=<-0.846 -0.208 -0.836>
   }
   n55 = "Transform (Modify) Bone - Bone: dyn_skirt_fr_03"(ComponentPose=@n54, Alpha=1, Translation="0,0,0", Rotation="-10.000,-10.000,-0.000", Scale="1,1,1") {
    rotationMode=BMM_Additive
    rotationSpace=BCS_BoneSpace
    translationMode=BMM_Additive
    translationSpace=BCS_BoneSpace
    alphaBoolBlend {blendOption=Linear}
    boneToModify {boneName=dyn_skirt_fr_03}
    rotation {pitch=-10, yaw=-10.0002}
   }
   "Transform (Modify) Bone - Bone: dyn_skirt_out_l_01"!(ComponentPose=@n55, Alpha=1, Translation:Vector="-0.000,-1.419,0.625", Rotation="0.000,-5.000,0.000", Scale:Vector="1,1,1") -> (Pose) {
    rotationMode=BMM_Additive
    rotationSpace=BCS_BoneSpace
    translationMode=BMM_Additive
    translationSpace=BCS_BoneSpace
    alphaBoolBlend {blendOption=Linear}
    boneToModify {boneName=dyn_skirt_out_l_01}
    rotation {yaw=-5}
    translation=<0 -1.4186 0.6253>
   }
   "Transform (Modify) Bone - Bone: dyn_skirt_fr_l_MSTR"(Alpha=1, Translation="7,2,2", Rotation="-10,6.5,4", Scale="1,1,1") {
    rotationMode=BMM_Additive
    rotationSpace=BCS_BoneSpace
    translationMode=BMM_Additive
    translationSpace=BCS_BoneSpace
    alphaBoolBlend {blendOption=Linear}
    boneToModify {boneName=dyn_skirt_fr_l_MSTR}
    rotation=<rot -10 6.5 4>
    translation=<7 2 2>
   }
   n56 = "Transform (Modify) Bone - Bone: dyn_skirt_fr_r_01"!(ComponentPose, Alpha=1, Translation="0,0,2.5", Rotation="4.994,-0.010,0.289") -> (Pose) {
    rotationMode=BMM_Replace
    rotationSpace=BCS_BoneSpace
    translationMode=BMM_Additive
    translationSpace=BCS_BoneSpace
    alphaBoolBlend {blendOption=Linear}
    alphaScaleBiasClamp {bInterpResult=true, interpSpeedDecreasing=5, interpSpeedIncreasing=7}
    boneToModify {boneName=dyn_skirt_fr_r_01}
    rotation=<rot 4.994 -0.01 0.289>
    translation {z=2.5}
   }
   n57 = "Transform (Modify) Bone - Bone: dyn_skirt_fr_MSTR"!(ComponentPose=@n56, Alpha=1, Translation="-1.156,-0.222,1.689", Rotation="33.360,-8.236,-10.777") -> (Pose) {
    rotationMode=BMM_Replace
    rotationSpace=BCS_BoneSpace
    translationMode=BMM_Additive
    translationSpace=BCS_BoneSpace
    alphaBoolBlend {blendOption=Linear}
    alphaScaleBiasClamp {bInterpResult=true, interpSpeedDecreasing=5, interpSpeedIncreasing=7}
    boneToModify {boneName=dyn_skirt_fr_MSTR}
    rotation=<rot 33.36 -8.2361 -10.7771>
    translation=<-1.156 -0.222 1.689>
   }
   "Transform (Modify) Bone - Bone: dyn_skirt_fr_l_01"!(ComponentPose=@n57, Alpha=1, Translation="0,0,0", Rotation="-5.423,-2.123,-4.622") -> (Pose) {
    rotationMode=BMM_Replace
    rotationSpace=BCS_BoneSpace
    translationMode=BMM_Additive
    translationSpace=BCS_BoneSpace
    alphaBoolBlend {blendOption=Linear}
    alphaScaleBiasClamp {bInterpResult=true, interpSpeedDecreasing=5, interpSpeedIncreasing=7}
    boneToModify {boneName=dyn_skirt_fr_l_01}
    rotation=<rot -5.423 -2.123 -4.622>
   }
   "Transform (Modify) Bone - Bone: dyn_skirt_fr_r_MSTR"(Alpha=1, Translation="5.764,1,2.618", Rotation="25,-10,5.496", Scale="1,1,1") {
    rotationMode=BMM_Additive
    rotationSpace=BCS_BoneSpace
    translationMode=BMM_Additive
    translationSpace=BCS_BoneSpace
    alphaBoolBlend {blendOption=Linear}
    alphaScaleBiasClamp {interpSpeedDecreasing=20, interpSpeedIncreasing=20}
    boneToModify {boneName=dyn_skirt_fr_r_MSTR}
    rotation=<rot 25 -10 5.496>
    translation=<5.764 1 2.618>
   }
   RigidBody() {
    bClampLinearTranslationLimitToRefPose=true
    bOverrideWorldGravity=true
    overridePhysicsAsset=&/F_MED_Octoberfest_Dyn_skirt
    simulationSpace=BaseBoneSpace
    alphaBoolBlend {blendOption=Linear}
    baseBoneRef {boneName=pelvis}
    componentAppliedLinearAccClamp=<3000 3000 3000>
    componentLinearAccScale=<0.4 0.3 0.3>
    componentLinearVelScale=<0.4 0.3 0.3>
    overrideWorldGravity {z=-980}
   }
   n58 = Get Right_Direct() -> (Right_Direct:double)
   n59 = "Transform (Modify) Bone - Bone: dyn_skirt_fr_l_01"!(
    ComponentPose=Local To Component(LocalPose=Use cached pose 'Rigid'())
    Alpha=@n58
    Translation="-10,-2,-0.9"
    Rotation="9.974,-15.236,-14.477"
    Scale:Vector="1.000,1.5,1.000"
   ) -> (Pose) {
    rotationMode=BMM_Replace
    rotationSpace=BCS_BoneSpace
    scaleMode=BMM_Replace
    scaleSpace=BCS_ParentBoneSpace
    translationMode=BMM_Replace
    translationSpace=BCS_BoneSpace
    alphaBoolBlend {blendOption=Linear}
    alphaScaleBiasClamp {bInterpResult=true, interpSpeedDecreasing=5, interpSpeedIncreasing=7}
    boneToModify {boneName=dyn_skirt_fr_l_01}
    rotation=<rot 9.974 -15.236 -14.477>
    scale {y=1.5}
    translation=<-10 -2 -0.9>
   }
   n60 = "Constraint - dyn_skirt_fr_l_02 (P)"(ComponentPose=@n59, Alpha=0.85, ConstraintWeights_0=1) {
    alphaBoolBlend {blendOption=Linear}
    alphaScaleBiasClamp {bInterpResult=true, interpSpeedDecreasing=5, interpSpeedIncreasing=7}
    boneToModify {boneName=dyn_skirt_fr_l_02}
    constraintSetup [{targetBone {boneName=thigh_l}}]
    constraintWeights [1]
   }
   n61 = "Constraint - dyn_skirt_fr_02 (PP)"!(ComponentPose=@n60, Alpha=@n58, ConstraintWeights_0=0.5, ConstraintWeights_1:float=0.5) -> (Pose) {
    alphaBoolBlend {blendOption=Linear}
    alphaScaleBiasClamp {bInterpResult=true, interpSpeedDecreasing=5, interpSpeedIncreasing=7}
    boneToModify {boneName=dyn_skirt_fr_02}
    constraintSetup [{targetBone {boneName=thigh_r}}, {targetBone {boneName=thigh_l}}]
    constraintWeights [1,1]
   }
   n62 = "Transform (Modify) Bone - Bone: dyn_skirt_fr_02"!(ComponentPose=@n61, Alpha=@n58, Translation="0,0,2", Rotation="5,5,0") -> (Pose) {
    rotationMode=BMM_Replace
    rotationSpace=BCS_BoneSpace
    translationMode=BMM_Replace
    translationSpace=BCS_BoneSpace
    alphaBoolBlend {blendOption=Linear}
    alphaScaleBiasClamp {bInterpResult=true, interpSpeedDecreasing=5, interpSpeedIncreasing=7}
    boneToModify {boneName=dyn_skirt_fr_02}
    rotation {pitch=5, yaw=5}
    translation {z=2}
   }
   n63 = "Transform (Modify) Bone - Bone: dyn_skirt_fr_MSTR"!(ComponentPose=@n62, Alpha=@n58, Rotation="11.700,-17,21.776") -> (Pose) {
    rotationMode=BMM_Replace
    rotationSpace=BCS_BoneSpace
    translationMode=BMM_Replace
    translationSpace=BCS_BoneSpace
    alphaBoolBlend {blendOption=Linear}
    alphaScaleBiasClamp {bInterpResult=true, interpSpeedDecreasing=5, interpSpeedIncreasing=7}
    boneToModify {boneName=dyn_skirt_fr_MSTR}
    rotation=<rot 11.7 -17 21.776>
   }
   n64 = "Constraint - dyn_skirt_fr_r_02 (P)"(ComponentPose=@n63, Alpha=@n58, ConstraintWeights_0=1) {
    alphaBoolBlend {blendOption=Linear}
    alphaScaleBiasClamp {bInterpResult=true, interpSpeedDecreasing=5, interpSpeedIncreasing=7}
    boneToModify {boneName=dyn_skirt_fr_r_02}
    constraintSetup [{targetBone {boneName=thigh_r}}]
    constraintWeights [1]
   }
   n65 = "Constraint - dyn_skirt_fr_l_03 (P)"(ComponentPose=@n64, Alpha=@n58, ConstraintWeights_0=1) {
    alphaBoolBlend {blendOption=Linear}
    alphaScaleBiasClamp {bClampResult=true, bInterpResult=true, clampMax=0.8, interpSpeedDecreasing=5, interpSpeedIncreasing=7}
    boneToModify {boneName=dyn_skirt_fr_l_03}
    constraintSetup [{targetBone {boneName=thigh_l}}]
    constraintWeights [1]
   }
   n66 = Look At - Bone: dyn_skirt_fr_02(ComponentPose=@n65, Alpha=@n58) {
    alphaBoolBlend {blendOption=Linear}
    alphaScaleBiasClamp {bInterpResult=true}
    boneToModify {boneName=dyn_skirt_fr_02}
    lookAtLocation=<-0.4286 89.3903 20.7677>
    lookAtTarget {boneReference {boneName=pelvis}}
    lookAt_Axis {axis=<1 0>}
   }
   n67 = Get Left_Direct() -> (Left_Direct:double)
   n68 = "Constraint - dyn_skirt_fr_r_02 (P)"(ComponentPose=@n66, Alpha=@n67, ConstraintWeights_0=1) {
    alphaBoolBlend {blendOption=Linear}
    boneToModify {boneName=dyn_skirt_fr_r_02}
    constraintSetup [{targetBone {boneName=thigh_r}}]
    constraintWeights [1]
   }
   n69 = "Constraint - dyn_skirt_fr_02 (P)"(ComponentPose=@n68, Alpha=@n67, ConstraintWeights_0=1) {
    alphaBoolBlend {blendOption=Linear}
    alphaScaleBiasClamp {bClampResult=true, bInterpResult=true, clampMax=0.5, interpSpeedDecreasing=5, interpSpeedIncreasing=7}
    boneToModify {boneName=dyn_skirt_fr_02}
    constraintSetup [{targetBone {boneName=dyn_skirt_fr_r_02}}]
    constraintWeights [1]
   }
   n70 = "Transform (Modify) Bone - Bone: dyn_skirt_fr_r_01"(ComponentPose=@n69, Alpha=@n67, Translation="0,0,5", Rotation="10,0,0") {
    rotationMode=BMM_Replace
    rotationSpace=BCS_BoneSpace
    scaleMode=BMM_Additive
    translationMode=BMM_Additive
    translationSpace=BCS_BoneSpace
    alphaBoolBlend {blendOption=Linear}
    alphaScaleBiasClamp {bInterpResult=true, interpSpeedDecreasing=5, interpSpeedIncreasing=7}
    boneToModify {boneName=dyn_skirt_fr_r_01}
    rotation {pitch=10}
    scale {y=2}
    translation {z=5}
   }
   n71 = "Transform (Modify) Bone - Bone: dyn_skirt_fr_MSTR"!(ComponentPose=@n70, Alpha=@n67, Translation:Vector="0,0,0", Rotation="25,0,0") -> (Pose) {
    rotationMode=BMM_Replace
    rotationSpace=BCS_BoneSpace
    translationMode=BMM_Additive
    translationSpace=BCS_BoneSpace
    alphaBoolBlend {blendOption=Linear}
    alphaScaleBiasClamp {bInterpResult=true, interpSpeedDecreasing=5, interpSpeedIncreasing=7}
    boneToModify {boneName=dyn_skirt_fr_MSTR}
    rotation {pitch=25}
   }
   n72 = "Transform (Modify) Bone - Bone: dyn_skirt_fr_l_01"!(ComponentPose=@n71, Alpha=@n67, Translation="0,0,0", Rotation="10,0,0") -> (Pose) {
    rotationMode=BMM_Replace
    rotationSpace=BCS_BoneSpace
    translationMode=BMM_Additive
    translationSpace=BCS_BoneSpace
    alphaBoolBlend {blendOption=Linear}
    alphaScaleBiasClamp {bInterpResult=true, interpSpeedDecreasing=5, interpSpeedIncreasing=7}
    boneToModify {boneName=dyn_skirt_fr_l_01}
    rotation {pitch=10}
   }
   n73 = Look At - Bone: dyn_skirt_fr_02(ComponentPose=@n72, Alpha=@n67) {
    alphaBoolBlend {blendOption=Linear}
    alphaScaleBiasClamp {bInterpResult=true, interpSpeedIncreasing=15}
    boneToModify {boneName=dyn_skirt_fr_02}
    lookAtLocation=<-87.6314 63.7316 -8.3999>
    lookAtTarget {boneReference {boneName=pelvis}}
    lookAt_Axis {axis=<1 0>}
   }
   n74 = Look At - Bone: dyn_skirt_fr_r_MSTR(ComponentPose=@n73, Alpha=@n67) {
    alphaBoolBlend {blendOption=Linear}
    alphaScaleBiasClamp {bInterpResult=true, interpSpeedIncreasing=15}
    boneToModify {boneName=dyn_skirt_fr_r_MSTR}
    lookAtLocation=<128.3383 27.5473 -5.3983>
    lookAtTarget {boneReference {boneName=thigh_r}}
    lookAt_Axis {axis=<1 0>}
   }
   n75 = Get crouch alpha()
   n76 = "Transform (Modify) Bone - Bone: dyn_skirt_fr_MSTR"!(ComponentPose=@n74, Alpha=@n75, Translation="0,0,0", Rotation="-0.455,4.957,5.424", Scale:Vector="1,1,1") -> (Pose) {
    rotationMode=BMM_Replace
    rotationSpace=BCS_BoneSpace
    translationMode=BMM_Replace
    translationSpace=BCS_BoneSpace
    alphaBoolBlend {blendOption=Linear}
    alphaScaleBiasClamp {interpSpeedDecreasing=5, interpSpeedIncreasing=7}
    boneToModify {boneName=dyn_skirt_fr_MSTR}
    rotation=<rot -0.455 4.957 5.424>
   }
   n77 = "Constraint - dyn_skirt_fr_l_01 (P)"(ComponentPose=@n76, Alpha=@n75, ConstraintWeights_0=1) {
    alphaBoolBlend {blendOption=Linear}
    alphaScaleBiasClamp {bClampResult=true, clampMax=0.8, interpSpeedDecreasing=5, interpSpeedIncreasing=7}
    boneToModify {boneName=dyn_skirt_fr_l_01}
    constraintSetup [{targetBone {boneName=thigh_l}}]
    constraintWeights [1]
   }
   n78 = "Transform (Modify) Bone - Bone: dyn_skirt_fr_l_01"!(ComponentPose=@n77, Alpha=@n75, Translation="0,0,0", Rotation="11.973,5.092,0.622", Scale:Vector="1,1,1") -> (Pose) {
    rotationMode=BMM_Replace
    rotationSpace=BCS_BoneSpace
    scaleMode=BMM_Replace
    scaleSpace=BCS_BoneSpace
    translationMode=BMM_Replace
    translationSpace=BCS_BoneSpace
    alphaBoolBlend {blendOption=Linear}
    alphaScaleBiasClamp {interpSpeedDecreasing=5, interpSpeedIncreasing=7}
    boneToModify {boneName=dyn_skirt_fr_l_01}
    rotation=<rot 11.973 5.092 0.622>
   }
   n79 = "Constraint - dyn_skirt_fr_l_02 (P)"(ComponentPose=@n78, Alpha=@n75, ConstraintWeights_0=1) {
    alphaBoolBlend {blendOption=Linear}
    alphaScaleBiasClamp {bClampResult=true, clampMax=0.8, interpSpeedDecreasing=5, interpSpeedIncreasing=7}
    boneToModify {boneName=dyn_skirt_fr_l_02}
    constraintSetup [{targetBone {boneName=thigh_l}}]
    constraintWeights [1]
   }
   n80 = "Constraint - dyn_skirt_fr_l_03 (P)"(ComponentPose=@n79, Alpha=@n75, ConstraintWeights_0=1) {
    alphaBoolBlend {blendOption=Linear}
    alphaScaleBiasClamp {clampMax=0.8, interpSpeedDecreasing=5, interpSpeedIncreasing=7}
    boneToModify {boneName=dyn_skirt_fr_l_03}
    constraintSetup [{targetBone {boneName=thigh_l}}]
    constraintWeights [1]
   }
   n81 = "Transform (Modify) Bone - Bone: dyn_skirt_fr_l_02"(ComponentPose=@n80, Alpha=@n75, Translation="0,0,0", Rotation="0,0,0", Scale="1,1,1") {
    rotationMode=BMM_Replace
    rotationSpace=BCS_BoneSpace
    scaleMode=BMM_Additive
    scaleSpace=BCS_BoneSpace
    translationMode=BMM_Replace
    translationSpace=BCS_BoneSpace
    alphaBoolBlend {blendOption=Linear}
    alphaScaleBiasClamp {interpSpeedDecreasing=5, interpSpeedIncreasing=7}
    boneToModify {boneName=dyn_skirt_fr_l_02}
   }
   n82 = Look At - Bone: dyn_skirt_fr_02(ComponentPose=@n81, Alpha=@n75) {
    alphaBoolBlend {blendOption=Linear}
    alphaScaleBiasClamp {bInterpResult=true}
    boneToModify {boneName=dyn_skirt_fr_02}
    lookAtLocation=<-13.6517 93.048 -6.574>
    lookAtTarget {boneReference {boneName=pelvis}}
    lookAt_Axis {axis=<1 0>}
   }
   n83 = Get Forward_Direct() -> (Forward_Direct:double)
   n84 = "Transform (Modify) Bone - Bone: dyn_skirt_fr_MSTR"(ComponentPose=@n82, Alpha=@n83, Translation="0,0,0", Rotation="12,0.000,0.000", Scale="1,1,1") {
    rotationMode=BMM_Replace
    rotationSpace=BCS_BoneSpace
    translationMode=BMM_Additive
    translationSpace=BCS_BoneSpace
    alphaBoolBlend {blendOption=Linear}
    alphaScaleBiasClamp {bInterpResult=true, interpSpeedDecreasing=5, interpSpeedIncreasing=7}
    boneToModify {boneName=dyn_skirt_fr_MSTR}
    rotation {pitch=12}
   }
   n85 = "Transform (Modify) Bone - Bone: dyn_skirt_fr_r_01"!(ComponentPose=@n84, Alpha=@n83, Translation="0,0,0", Rotation="-7,0.000,0.000", Scale:Vector="1,1,1") -> (Pose) {
    rotationMode=BMM_Additive
    rotationSpace=BCS_BoneSpace
    translationMode=BMM_Additive
    translationSpace=BCS_BoneSpace
    alphaBoolBlend {blendOption=Linear}
    alphaScaleBiasClamp {bInterpResult=true, interpSpeedDecreasing=5, interpSpeedIncreasing=7}
    boneToModify {boneName=dyn_skirt_fr_r_01}
    rotation {pitch=-7}
   }
   n86 = "Transform (Modify) Bone - Bone: dyn_skirt_fr_l_01"(ComponentPose=@n85, Alpha=@n83, Translation="0,0,0", Rotation="12,0.000,0.000", Scale="1,1,1") {
    rotationMode=BMM_Additive
    rotationSpace=BCS_BoneSpace
    translationMode=BMM_Additive
    translationSpace=BCS_BoneSpace
    alphaBoolBlend {blendOption=Linear}
    alphaScaleBiasClamp {bInterpResult=true, interpSpeedDecreasing=5, interpSpeedIncreasing=7}
    boneToModify {boneName=dyn_skirt_fr_l_01}
    rotation {pitch=12}
   }
   n87 = "Constraint - dyn_skirt_fr_l_02 (P)"(ComponentPose=@n86, Alpha=@n83, ConstraintWeights_0=1) {
    alphaBoolBlend {blendOption=Linear}
    alphaScaleBiasClamp {bClampResult=true, bInterpResult=true, clampMax=0.6, interpSpeedDecreasing=5, interpSpeedIncreasing=7}
    boneToModify {boneName=dyn_skirt_fr_l_02}
    constraintSetup [{targetBone {boneName=thigh_l}}]
    constraintWeights [1]
   }
   n88 = "Constraint - dyn_skirt_fr_02 (P)"(ComponentPose=@n87, Alpha=@n83, ConstraintWeights_0=1) {
    alphaBoolBlend {blendOption=Linear}
    alphaScaleBiasClamp {bClampResult=true, bInterpResult=true, clampMax=0.5, interpSpeedDecreasing=5, interpSpeedIncreasing=7}
    boneToModify {boneName=dyn_skirt_fr_02}
    constraintSetup [{targetBone {boneName=thigh_r}}]
    constraintWeights [1]
   }
   n89 = "Constraint - dyn_skirt_fr_l_03 (P)"(ComponentPose=@n88, Alpha=@n83, ConstraintWeights_0=1) {
    alphaBoolBlend {blendOption=Linear}
    alphaScaleBiasClamp {bClampResult=true, bInterpResult=true, clampMax=0.6, interpSpeedDecreasing=5, interpSpeedIncreasing=7}
    boneToModify {boneName=dyn_skirt_fr_l_03}
    constraintSetup [{targetBone {boneName=thigh_l}}]
    constraintWeights [1]
   }
   n90 = "Transform (Modify) Bone - Bone: dyn_skirt_out_l_MSTR"(ComponentPose=@n89, Alpha=1, Translation="0,0,0", Rotation="-10.000,0.000,0.000", Scale="1,1,1") {
    rotationMode=BMM_Additive
    rotationSpace=BCS_BoneSpace
    translationMode=BMM_Additive
    translationSpace=BCS_BoneSpace
    alphaBoolBlend {blendOption=Linear}
    alphaScaleBiasClamp {bInterpResult=true}
    boneToModify {boneName=dyn_skirt_out_l_MSTR}
    rotation {pitch=-10}
   }
   n91 = "Transform (Modify) Bone - Bone: dyn_skirt_out_r_MSTR"(ComponentPose=@n90, Alpha=1, Translation="0,0,0", Rotation="-10.000,0.000,0.000", Scale="1,1,1") {
    rotationMode=BMM_Additive
    rotationSpace=BCS_BoneSpace
    translationMode=BMM_Additive
    translationSpace=BCS_BoneSpace
    alphaBoolBlend {blendOption=Linear}
    alphaScaleBiasClamp {bInterpResult=true}
    boneToModify {boneName=dyn_skirt_out_r_MSTR}
    rotation {pitch=-10}
   }
   n92 = "Transform (Modify) Bone - Bone: dyn_skirt_bk_01"(ComponentPose=@n91, Alpha=1, Translation="0,0,0", Rotation="-10.000,0.000,0.000", Scale="1,1,1") {
    rotationMode=BMM_Additive
    rotationSpace=BCS_BoneSpace
    translationMode=BMM_Additive
    translationSpace=BCS_BoneSpace
    alphaBoolBlend {blendOption=Linear}
    alphaScaleBiasClamp {bInterpResult=true}
    boneToModify {boneName=dyn_skirt_bk_01}
    rotation {pitch=-10}
   }
   n93 = "Transform (Modify) Bone - Bone: dyn_skirt_fr_MSTR"(ComponentPose=@n92, Alpha=1, Translation="0,0,0", Rotation="-15.000,0.000,0.000", Scale="1,1,1") {
    rotationMode=BMM_Additive
    rotationSpace=BCS_BoneSpace
    translationMode=BMM_Additive
    translationSpace=BCS_BoneSpace
    alphaBoolBlend {blendOption=Linear}
    alphaScaleBiasClamp {bInterpResult=true, interpSpeedDecreasing=15, interpSpeedIncreasing=15}
    boneToModify {boneName=dyn_skirt_fr_MSTR}
    rotation {pitch=-15}
   }
   n94 = Get Snap_Skirt_front() -> (Snap_Skirt_front:double)
   n95 = "Constraint - dyn_skirt_fr_r_01 (P)"(ComponentPose=@n93, Alpha=@n94, ConstraintWeights_0=1) {
    alphaBoolBlend {blendOption=Linear}
    boneToModify {boneName=dyn_skirt_fr_r_01}
    constraintSetup [{targetBone {boneName=thigh_r}}]
    constraintWeights [1]
   }
   n96 = "Constraint - dyn_skirt_fr_01 (PP)"(ComponentPose=@n95, Alpha=@n94, ConstraintWeights_0=0.5, ConstraintWeights_1=0.5) {
    alphaBoolBlend {blendOption=Linear}
    boneToModify {boneName=dyn_skirt_fr_01}
    constraintSetup [{targetBone {boneName=thigh_r}}, {targetBone {boneName=thigh_l}}]
    constraintWeights [1,1]
   }
   n97 = "Constraint - dyn_skirt_fr_l_01 (P)"(ComponentPose=@n96, Alpha=@n94, ConstraintWeights_0=1) {
    alphaBoolBlend {blendOption=Linear}
    boneToModify {boneName=dyn_skirt_fr_l_01}
    constraintSetup [{targetBone {boneName=thigh_l}}]
    constraintWeights [1]
   }
   n98 = "Constraint - dyn_skirt_fr_02 (P)"(ComponentPose=@n97, Alpha=@n94, ConstraintWeights_0=1) {
    alphaBoolBlend {blendOption=Linear}
    boneToModify {boneName=dyn_skirt_fr_02}
    constraintSetup [{targetBone {boneName=thigh_r}}]
    constraintWeights [1]
   }
   n99 = "Constraint - dyn_skirt_fr_03 (P)"(ComponentPose=@n98, Alpha=@n94, ConstraintWeights_0=1) {
    alphaBoolBlend {blendOption=Linear}
    boneToModify {boneName=dyn_skirt_fr_03}
    constraintSetup [{targetBone {boneName=thigh_r}}]
    constraintWeights [1]
   }
   n100 = "Constraint - dyn_skirt_fr_04 (P)"(ComponentPose=@n99, Alpha=@n94, ConstraintWeights_0=1) {
    alphaBoolBlend {blendOption=Linear}
    boneToModify {boneName=dyn_skirt_fr_04}
    constraintSetup [{targetBone {boneName=thigh_r}}]
    constraintWeights [1]
   }
   n101 = "Constraint - dyn_skirt_fr_l_02 (P)"(ComponentPose=@n100, Alpha=@n94, ConstraintWeights_0=1) {
    alphaBoolBlend {blendOption=Linear}
    boneToModify {boneName=dyn_skirt_fr_l_02}
    constraintSetup [{targetBone {boneName=thigh_l}}]
    constraintWeights [1]
   }
   n102 = "Constraint - dyn_skirt_fr_l_03 (P)"(ComponentPose=@n101, Alpha=@n94, ConstraintWeights_0=1) {
    alphaBoolBlend {blendOption=Linear}
    boneToModify {boneName=dyn_skirt_fr_l_03}
    constraintSetup [{targetBone {boneName=thigh_l}}]
    constraintWeights [1]
   }
   n103 = "Constraint - dyn_skirt_fr_l_04 (P)"(ComponentPose=@n102, Alpha=@n94, ConstraintWeights_0=1) {
    alphaBoolBlend {blendOption=Linear}
    boneToModify {boneName=dyn_skirt_fr_l_04}
    constraintSetup [{targetBone {boneName=thigh_l}}]
    constraintWeights [1]
   }
   n104 = "Constraint - dyn_skirt_fr_02 (PP)"(ComponentPose=@n103, Alpha=@n94, ConstraintWeights_0=0.5, ConstraintWeights_1=0.5) {
    alphaBoolBlend {blendOption=Linear}
    boneToModify {boneName=dyn_skirt_fr_02}
    constraintSetup [{targetBone {boneName=thigh_l}}, {targetBone {boneName=thigh_r}}]
    constraintWeights [1,1]
   }
   n105 = "Constraint - dyn_skirt_fr_03 (PP)"!(ComponentPose=@n104, Alpha=@n94, ConstraintWeights_0=0.5, ConstraintWeights_1:float=0.5) -> (Pose) {
    alphaBoolBlend {blendOption=Linear}
    boneToModify {boneName=dyn_skirt_fr_03}
    constraintSetup [{targetBone {boneName=thigh_l}}, {targetBone {boneName=thigh_r}}]
    constraintWeights [1,1]
   }
   n106 = "Constraint - dyn_skirt_fr_04 (PP)"!(ComponentPose=@n105, Alpha=@n94, ConstraintWeights_0=0.5, ConstraintWeights_1:float=0.5) -> (Pose) {
    alphaBoolBlend {blendOption=Linear}
    boneToModify {boneName=dyn_skirt_fr_04}
    constraintSetup [{targetBone {boneName=thigh_l}}, {targetBone {boneName=thigh_r}}]
    constraintWeights [1,1]
   }
   Output Pose(Result=Component To Local(ComponentPose=@n106)) {layerGroup=None}
   n107 = "Constraint - dyn_skirt_fr_04 (R)"(ComponentPose=Local To Component(LocalPose=Copy Pose From Mesh() {bUseAttachedParent=true}), Alpha=1, ConstraintWeights_0=1) {
    alphaBoolBlend {blendOption=Linear}
    boneToModify {boneName=dyn_skirt_fr_04}
    constraintSetup [{transformType=Rotation, targetBone {boneName=dyn_skirt_fr_02}}]
    constraintWeights [1]
   }
   n108 = "Constraint - dyn_skirt_fr_l_04 (R)"(ComponentPose=@n107, Alpha=1, ConstraintWeights_0=1) {
    alphaBoolBlend {blendOption=Linear}
    boneToModify {boneName=dyn_skirt_fr_l_04}
    constraintSetup [{transformType=Rotation, targetBone {boneName=dyn_skirt_fr_l_02}}]
    constraintWeights [1]
   }
   n109 = "Constraint - dyn_skirt_fr_r_04 (P)"(ComponentPose=@n108, Alpha=1, ConstraintWeights_0=1) {
    alphaBoolBlend {blendOption=Linear}
    boneToModify {boneName=dyn_skirt_fr_r_04}
    constraintSetup [{targetBone {boneName=dyn_skirt_fr_r_02}}]
    constraintWeights [1]
   }
   n110 = "Constraint - dyn_skirt_fr_MSTR (PP)"!(ComponentPose=@n109, Alpha=0.8, ConstraintWeights_0=0, ConstraintWeights_1:float=1) -> (Pose) {
    alphaBoolBlend {blendOption=Linear}
    boneToModify {boneName=dyn_skirt_fr_MSTR}
    constraintSetup [{targetBone {boneName=pelvis}}, {targetBone {boneName=thigh_l}}]
    constraintWeights [1,1]
   }
   n111 = "Constraint - dyn_skirt_fr_02 (PP)"(ComponentPose=@n110, Alpha=0.4, ConstraintWeights_0=0, ConstraintWeights_1=1) {
    alphaBoolBlend {blendOption=Linear}
    boneToModify {boneName=dyn_skirt_fr_02}
    constraintSetup [{targetBone {boneName=pelvis}}, {targetBone {boneName=thigh_l}}]
    constraintWeights [1,1]
   }
   n112 = "Constraint - dyn_skirt_fr_03 (PP)"(ComponentPose=@n111, Alpha=0.7, ConstraintWeights_0=0, ConstraintWeights_1=1) {
    alphaBoolBlend {blendOption=Linear}
    boneToModify {boneName=dyn_skirt_fr_03}
    constraintSetup [{targetBone {boneName=pelvis}}, {targetBone {boneName=thigh_l}}]
    constraintWeights [1,1]
   }
   n113 = "Constraint - dyn_skirt_fr_04 (PP)"(ComponentPose=@n112, Alpha=0.7, ConstraintWeights_0=0, ConstraintWeights_1=1) {
    alphaBoolBlend {blendOption=Linear}
    boneToModify {boneName=dyn_skirt_fr_04}
    constraintSetup [{targetBone {boneName=pelvis}}, {targetBone {boneName=thigh_l}}]
    constraintWeights [1,1]
   }
   n114 = "Constraint - dyn_skirt_fr_l_MSTR (PP)"!(ComponentPose=@n113, Alpha=0.8, ConstraintWeights_0=0, ConstraintWeights_1:float=1) -> (Pose) {
    alphaBoolBlend {blendOption=Linear}
    boneToModify {boneName=dyn_skirt_fr_l_MSTR}
    constraintSetup [{targetBone {boneName=pelvis}}, {targetBone {boneName=thigh_l}}]
    constraintWeights [1,1]
   }
   n115 = "Constraint - dyn_skirt_fr_l_02 (PP)"!(ComponentPose=@n114, Alpha=0.4, ConstraintWeights_0=0, ConstraintWeights_1:float=1) -> (Pose) {
    alphaBoolBlend {blendOption=Linear}
    boneToModify {boneName=dyn_skirt_fr_l_02}
    constraintSetup [{targetBone {boneName=pelvis}}, {targetBone {boneName=thigh_l}}]
    constraintWeights [1,1]
   }
   n116 = "Constraint - dyn_skirt_fr_l_03 (PP)"!(ComponentPose=@n115, Alpha=0.7, ConstraintWeights_0=0, ConstraintWeights_1:float=1) -> (Pose) {
    alphaBoolBlend {blendOption=Linear}
    boneToModify {boneName=dyn_skirt_fr_l_03}
    constraintSetup [{targetBone {boneName=pelvis}}, {targetBone {boneName=thigh_l}}]
    constraintWeights [1,1]
   }
   n117 = "Constraint - dyn_skirt_fr_l_04 (PP)"!(ComponentPose=@n116, Alpha=0.7, ConstraintWeights_0=0, ConstraintWeights_1:float=1) -> (Pose) {
    alphaBoolBlend {blendOption=Linear}
    boneToModify {boneName=dyn_skirt_fr_l_04}
    constraintSetup [{targetBone {boneName=pelvis}}, {targetBone {boneName=thigh_l}}]
    constraintWeights [1,1]
   }
   n118 = "Constraint - dyn_skirt_fr_r_MSTR (PP)"!(ComponentPose=@n117, Alpha=0.8, ConstraintWeights_0=0, ConstraintWeights_1:float=1) -> (Pose) {
    alphaBoolBlend {blendOption=Linear}
    boneToModify {boneName=dyn_skirt_fr_r_MSTR}
    constraintSetup [{targetBone {boneName=thigh_twist_01_r}}, {targetBone {boneName=thigh_r}}]
    constraintWeights [1,1]
   }
   n119 = "Constraint - dyn_skirt_fr_r_02 (PP)"!(ComponentPose=@n118, Alpha=0.8, ConstraintWeights_0=0, ConstraintWeights_1:float=1) -> (Pose) {
    alphaBoolBlend {blendOption=Linear}
    boneToModify {boneName=dyn_skirt_fr_r_02}
    constraintSetup [{targetBone {boneName=pelvis}}, {targetBone {boneName=thigh_r}}]
    constraintWeights [1,1]
   }
   n120 = "Constraint - dyn_skirt_fr_r_02 (PPP)"!(ComponentPose=@n119, Alpha=0.8, ConstraintWeights_0=0.7799, ConstraintWeights_1:float=0.0297, ConstraintWeights_2:float=0.1904) -> (Pose) {
    alphaBoolBlend {blendOption=Linear}
    boneToModify {boneName=dyn_skirt_fr_r_02}
    constraintSetup [{targetBone {boneName=thigh_r}}, {targetBone {boneName=pelvis}}, {targetBone {boneName=thigh_twist_01_r}}]
    constraintWeights [1,1,1]
   }
   n121 = "Constraint - dyn_skirt_fr_r_03 (PP)"!(ComponentPose=@n120, Alpha=0.7, ConstraintWeights_0=0, ConstraintWeights_1:float=1) -> (Pose) {
    alphaBoolBlend {blendOption=Linear}
    boneToModify {boneName=dyn_skirt_fr_r_03}
    constraintSetup [{targetBone {boneName=pelvis}}, {targetBone {boneName=thigh_r}}]
    constraintWeights [1,1]
   }
   n122 = "Constraint - dyn_skirt_fr_r_04 (PP)"!(ComponentPose=@n121, Alpha=0.7, ConstraintWeights_0=0, ConstraintWeights_1:float=1) -> (Pose) {
    alphaBoolBlend {blendOption=Linear}
    boneToModify {boneName=dyn_skirt_fr_r_04}
    constraintSetup [{targetBone {boneName=pelvis}}, {targetBone {boneName=thigh_r}}]
    constraintWeights [1,1]
   }
   n123 = "Constraint - dyn_skirt_out_l_04 (P)"(ComponentPose=@n122, Alpha=1, ConstraintWeights_0=1) {
    alphaBoolBlend {blendOption=Linear}
    boneToModify {boneName=dyn_skirt_out_l_04}
    constraintSetup [{targetBone {boneName=dyn_skirt_out_l_02}}]
    constraintWeights [1]
   }
   n124 = "Constraint - dyn_skirt_bk_l_04 (R)"(ComponentPose=@n123, Alpha=1, ConstraintWeights_0=1) {
    alphaBoolBlend {blendOption=Linear}
    boneToModify {boneName=dyn_skirt_bk_l_04}
    constraintSetup [{transformType=Rotation, targetBone {boneName=dyn_skirt_bk_l_02}}]
    constraintWeights [1]
   }
   n125 = "Constraint - dyn_skirt_bk_r_04 (R)"(ComponentPose=@n124, Alpha=0.8, ConstraintWeights_0=1) {
    alphaBoolBlend {blendOption=Linear}
    boneToModify {boneName=dyn_skirt_bk_r_04}
    constraintSetup [{transformType=Rotation, targetBone {boneName=dyn_skirt_bk_r_02}}]
    constraintWeights [1]
   }
   n126 = "Constraint - dyn_skirt_out_r_04 (P)"(ComponentPose=@n125, Alpha=1, ConstraintWeights_0=1) {
    alphaBoolBlend {blendOption=Linear}
    boneToModify {boneName=dyn_skirt_out_r_04}
    constraintSetup [{targetBone {boneName=dyn_skirt_out_r_02}}]
    constraintWeights [1]
   }
   n127 = Anim Dynamics - Bone: dyn_choker_l(ComponentPose=@n126, Alpha=1, ExternalForce="0,0,0") {
    angularDampingOverride=0.9999
    angularSpringConstant=75
    bAngularSpring=true
    bGravityOverrideInSimSpace=true
    bOverrideAngularDamping=true
    bUseGravityOverride=true
    alphaBoolBlend {blendOption=Linear}
    boundBone {boneName=dyn_choker_l}
    componentLinearAccScale=<1.2 1.2 1.2>
    componentLinearVelScale=<1.2 1.2 1.2>
    gravityOverride {z=-980}
    physicsBodyDefinitions [{
      sphereCollisionRadius=0
      boundBone {boneName=dyn_choker_l}
      boxExtents=<20 20 8>
      constraintSetup {angularTargetAxis=AxisZ, twistAxis=AxisZ, angularLimitsMax=<25 0 25>, angularLimitsMin=<-25 -25 -25>, angularTarget=<0 0 0>}
      localJointOffset=<0 0 -5>
     }]
   }
   n128 = Anim Dynamics - Bone: dyn_choker_r(ComponentPose=@n127, Alpha=1, ExternalForce="0,0,0") {
    angularDampingOverride=0.9999
    angularSpringConstant=75
    bAngularSpring=true
    bGravityOverrideInSimSpace=true
    bOverrideAngularDamping=true
    bUseGravityOverride=true
    alphaBoolBlend {blendOption=Linear}
    boundBone {boneName=dyn_choker_r}
    componentLinearAccScale=<1.2 1.2 1.2>
    componentLinearVelScale=<1.2 1.2 1.2>
    gravityOverride {z=-980}
    physicsBodyDefinitions [{
      sphereCollisionRadius=0
      boundBone {boneName=dyn_choker_r}
      boxExtents=<20 20 8>
      constraintSetup {angularTargetAxis=AxisZ, twistAxis=AxisZ, angularLimitsMax=<25 0 25>, angularLimitsMin=<-25 -25 -25>, angularTarget=<0 0 0>}
      localJointOffset=<0 0 -5>
     }]
   }
   n129 = Anim Dynamics - Bone: dyn_chestbow_string_l(ComponentPose=@n128, Alpha=0.8, ExternalForce="0,0,0") {
    angularDampingOverride=0.9999
    angularSpringConstant=75
    bAngularSpring=true
    bGravityOverrideInSimSpace=true
    bOverrideAngularDamping=true
    bUseGravityOverride=true
    simulationSpace=RootRelative
    alphaBoolBlend {blendOption=Linear}
    boundBone {boneName=dyn_chestbow_string_l}
    componentLinearAccScale=<1.2 1.2 1.2>
    componentLinearVelScale=<1.2 1.2 1.2>
    gravityOverride {z=-980}
    physicsBodyDefinitions [{
      sphereCollisionRadius=0
      boundBone {boneName=dyn_chestbow_string_l}
      boxExtents=<25 25 5>
      constraintSetup {angularTargetAxis=AxisZ, twistAxis=AxisZ, angularLimitsMax=<25 15 25>, angularLimitsMin=<-25 -25 -25>, angularTarget=<0 0 0>}
      localJointOffset=<0 0 -2>
     }]
   }
   n130 = Anim Dynamics - Bone: dyn_chestbow_string_r(ComponentPose=@n129, Alpha=0.8, ExternalForce="0,0,0") {
    angularDampingOverride=0.9999
    angularSpringConstant=75
    bAngularSpring=true
    bGravityOverrideInSimSpace=true
    bOverrideAngularDamping=true
    bUseGravityOverride=true
    simulationSpace=RootRelative
    alphaBoolBlend {blendOption=Linear}
    boundBone {boneName=dyn_chestbow_string_r}
    componentLinearAccScale=<1.2 1.2 1.2>
    componentLinearVelScale=<1.2 1.2 1.2>
    gravityOverride {z=-980}
    physicsBodyDefinitions [{
      sphereCollisionRadius=0
      boundBone {boneName=dyn_chestbow_string_r}
      boxExtents=<15 15 5>
      constraintSetup {angularTargetAxis=AxisZ, twistAxis=AxisZ, angularLimitsMax=<25 10 25>, angularLimitsMin=<-25 -30 -25>, angularTarget=<0 0 0>}
      localJointOffset=<0 0 -2>
     }]
   }
   n131 = Anim Dynamics - Bone: dyn_bellybow_string_l_01(ComponentPose=@n130, Alpha=0.5, ExternalForce="0,0,0") {
    angularDampingOverride=0.9999
    angularSpringConstant=75
    bAngularSpring=true
    bGravityOverrideInSimSpace=true
    bOverrideAngularDamping=true
    bUseGravityOverride=true
    simulationSpace=RootRelative
    alphaBoolBlend {blendOption=Linear}
    boundBone {boneName=dyn_bellybow_string_l_01}
    componentLinearAccScale=<1.2 1.2 1.2>
    componentLinearVelScale=<1.2 1.2 1.2>
    gravityOverride {z=-980}
    physicsBodyDefinitions [{
      sphereCollisionRadius=0
      boundBone {boneName=dyn_bellybow_string_l_01}
      boxExtents=<10 10 5>
      constraintSetup {
       angularTargetAxis=AxisZ
       twistAxis=AxisZ
       angularLimitsMax=<25 80 25>
       angularLimitsMin=<-25 -35 -25>
       angularTarget=<0 0 0>
       linearAxesMax=<2 1 0>
       linearAxesMin=<-2 -1 0>
      }
      localJointOffset=<0 0 -3>
     }]
   }
   n132 = Anim Dynamics - Bone: dyn_bellybow_string_l_02(ComponentPose=@n131, Alpha=0.5, ExternalForce="0,0,0") {
    angularDampingOverride=0.9999
    angularSpringConstant=75
    bAngularSpring=true
    bGravityOverrideInSimSpace=true
    bOverrideAngularDamping=true
    bUseGravityOverride=true
    simulationSpace=RootRelative
    alphaBoolBlend {blendOption=Linear}
    boundBone {boneName=dyn_bellybow_string_l_02}
    componentLinearAccScale=<1.2 1.2 1.2>
    componentLinearVelScale=<1.2 1.2 1.2>
    gravityOverride {z=-980}
    physicsBodyDefinitions [{
      sphereCollisionRadius=0
      boundBone {boneName=dyn_bellybow_string_l_02}
      boxExtents=<10 10 5>
      constraintSetup {angularTargetAxis=AxisZ, twistAxis=AxisZ, angularLimitsMax=<25 90 25>, angularLimitsMin=<-25 -10 -25>, angularTarget=<0 0 0>}
      localJointOffset=<0 0 -2>
     }]
   }
   n133 = Anim Dynamics - Bone: dyn_bellybow_string_r_01(ComponentPose=@n132, Alpha=0.5, ExternalForce="0,0,0") {
    angularDampingOverride=0.9999
    angularSpringConstant=75
    bAngularSpring=true
    bGravityOverrideInSimSpace=true
    bOverrideAngularDamping=true
    bUseGravityOverride=true
    simulationSpace=RootRelative
    alphaBoolBlend {blendOption=Linear}
    boundBone {boneName=dyn_bellybow_string_r_01}
    componentLinearAccScale=<1.2 1.2 1.2>
    componentLinearVelScale=<1.2 1.2 1.2>
    gravityOverride {z=-980}
    physicsBodyDefinitions [{
      sphereCollisionRadius=0
      boundBone {boneName=dyn_bellybow_string_r_01}
      boxExtents=<10 10 5>
      constraintSetup {
       angularTargetAxis=AxisZ
       twistAxis=AxisZ
       angularLimitsMax=<25 80 25>
       angularLimitsMin=<-25 -35 -25>
       angularTarget=<0 0 0>
       linearAxesMax=<2 1 0>
       linearAxesMin=<-2 -1 0>
      }
      localJointOffset=<0 0 -2>
     }]
   }
   n134 = Anim Dynamics - Bone: dyn_bellybow_string_r_02(ComponentPose=@n133, Alpha=0.5, ExternalForce="0,0,0") {
    angularDampingOverride=0.9999
    angularSpringConstant=75
    bAngularSpring=true
    bGravityOverrideInSimSpace=true
    bOverrideAngularDamping=true
    bUseGravityOverride=true
    simulationSpace=RootRelative
    alphaBoolBlend {blendOption=Linear}
    boundBone {boneName=dyn_bellybow_string_r_02}
    componentLinearAccScale=<1.2 1.2 1.2>
    componentLinearVelScale=<1.2 1.2 1.2>
    gravityOverride {z=-980}
    physicsBodyDefinitions [{
      sphereCollisionRadius=0
      boundBone {boneName=dyn_bellybow_string_r_02}
      boxExtents=<10 10 5>
      constraintSetup {angularTargetAxis=AxisZ, twistAxis=AxisZ, angularLimitsMax=<25 90 25>, angularLimitsMin=<-25 -10 -25>, angularTarget=<0 0 0>}
      localJointOffset=<0 0 0>
     }]
   }
   n135 = Anim Dynamics - Bone: dyn_buttbow_string_l_01(ComponentPose=@n134, Alpha=1, ExternalForce="0,0,0") {
    angularDampingOverride=1
    angularSpringConstant=1500
    bAngularSpring=true
    bGravityOverrideInSimSpace=true
    bOverrideAngularDamping=true
    bUseGravityOverride=true
    gravityScale=0.35
    alphaBoolBlend {blendOption=Linear}
    alphaScaleBiasClamp {scale=0.4}
    boundBone {boneName=dyn_buttbow_string_l_01}
    componentAppliedLinearAccClamp {z=10000}
    componentLinearAccScale=<1.2 1.2 0.46>
    componentLinearVelScale=<1.2 1.2 0.46>
    gravityOverride {z=-980}
    physicsBodyDefinitions [{
      sphereCollisionRadius=0
      boundBone {boneName=dyn_buttbow_string_l_01}
      boxExtents=<20 20 8>
      constraintSetup {angularTargetAxis=AxisZ, twistAxis=AxisZ, angularLimitsMax=<30 5 2>, angularLimitsMin=<-40 -30 -1>, angularTarget=<0 0 0>}
      localJointOffset=<0 0 -10>
     }]
   }
   n136 = Anim Dynamics - Bone: dyn_buttbow_string_r_01(ComponentPose=@n135, Alpha=1, ExternalForce="0,0,0") {
    angularDampingOverride=1
    angularSpringConstant=1500
    bAngularSpring=true
    bGravityOverrideInSimSpace=true
    bOverrideAngularDamping=true
    bUseGravityOverride=true
    gravityScale=0.35
    alphaBoolBlend {blendOption=Linear}
    alphaScaleBiasClamp {scale=0.4}
    boundBone {boneName=dyn_buttbow_string_r_01}
    componentAppliedLinearAccClamp {z=10000}
    componentLinearAccScale=<1.2 1.2 0.46>
    componentLinearVelScale=<1.2 1.2 0.46>
    gravityOverride {z=-980}
    physicsBodyDefinitions [{
      sphereCollisionRadius=0
      boundBone {boneName=dyn_buttbow_string_r_01}
      boxExtents=<20 20 8>
      constraintSetup {angularTargetAxis=AxisZ, twistAxis=AxisZ, angularLimitsMax=<30 5 1>, angularLimitsMin=<-40 -30 -2>, angularTarget=<0 0 0>}
      localJointOffset=<0 0 -10>
     }]
   }
   n137 = "Constraint - dyn_buttbow_base (P)"!(ComponentPose=@n136, Alpha=1, ConstraintWeights_0=1, ConstraintWeights_1:float=0) -> (Pose) {
    alphaBoolBlend {blendOption=Linear}
    boneToModify {boneName=dyn_buttbow_base}
    constraintSetup [{targetBone {boneName=pelvis}}]
    constraintWeights [1,0]
   }
   n138 = "Constraint - dyn_bellybow_base (PPPP)"!(
    ComponentPose=@n137
    Alpha=1
    ConstraintWeights_0=0.0148
    ConstraintWeights_1:float=0.0464
    ConstraintWeights_2:float=0.0994
    ConstraintWeights_3:float=0.8393
   ) -> (Pose) {
    alphaBoolBlend {blendOption=Linear}
    boneToModify {boneName=dyn_bellybow_base}
    constraintSetup [{targetBone {boneName=spine_01}}, {targetBone {boneName=spine_02}}, {targetBone {boneName=spine_03}}, {targetBone {boneName=spine_04}}]
    constraintWeights [1,1,1,1]
   }
   n139 = "Constraint - dyn_chestbow_base (PP)"!(ComponentPose=@n138, Alpha=1, ConstraintWeights_0=0.7536, ConstraintWeights_1:float=0.2464) -> (Pose) {
    alphaBoolBlend {blendOption=Linear}
    boneToModify {boneName=dyn_chestbow_base}
    constraintSetup [{targetBone {boneName=spine_04}}, {targetBone {boneName=spine_05}}]
    constraintWeights [1,1]
   }
   n140 = "Constraint - dyn_choker_l (PPPPP)"!(
    ComponentPose=@n139
    Alpha=0.5
    ConstraintWeights_0=0.0129
    ConstraintWeights_1:float=0.0409
    ConstraintWeights_2:float=0.1054
    ConstraintWeights_3:float=0.8341
    ConstraintWeights_4:float=0.0066
   ) -> (Pose) {
    alphaBoolBlend {blendOption=Linear}
    boneToModify {boneName=dyn_choker_l}
    constraintSetup [
     {targetBone {boneName=clavicle_r}}
     {targetBone {boneName=neck_02}}
     {targetBone {boneName=neck_01}}
     {targetBone {boneName=spine_05}}
     {targetBone {boneName=clavicle_l}}
    ]
    constraintWeights [1,1,1,1,1]
   }
   n141 = "Constraint - dyn_choker_r (PPPPP)"!(
    ComponentPose=@n140
    Alpha=0.5
    ConstraintWeights_0=0.0129
    ConstraintWeights_1:float=0.0409
    ConstraintWeights_2:float=0.1054
    ConstraintWeights_3:float=0.8341
    ConstraintWeights_4:float=0.0066
   ) -> (Pose) {
    alphaBoolBlend {blendOption=Linear}
    boneToModify {boneName=dyn_choker_r}
    constraintSetup [
     {targetBone {boneName=clavicle_r}}
     {targetBone {boneName=neck_02}}
     {targetBone {boneName=neck_01}}
     {targetBone {boneName=spine_05}}
     {targetBone {boneName=clavicle_l}}
    ]
    constraintWeights [1,1,1,1,1]
   }
   n142 = Get thigh_rotation_l() -> (thigh_rotation_l:double)
   n143 = "Transform (Modify) Bone - Bone: dyn_skirt_fr_MSTR"(ComponentPose=@n141, Alpha=@n142, Translation="6,4,0.1", Rotation="-25,4,23", Scale="1,1,1") {
    rotationMode=BMM_Replace
    rotationSpace=BCS_BoneSpace
    translationMode=BMM_Replace
    translationSpace=BCS_BoneSpace
    alphaBoolBlend {blendOption=Linear}
    boneToModify {boneName=dyn_skirt_fr_MSTR}
    rotation=<rot -25 4 23>
    translation=<6 4 0.1>
   }
   n144 = "Transform (Modify) Bone - Bone: dyn_skirt_fr_l_MSTR"(ComponentPose=@n143, Alpha=@n142, Translation="7,2,2", Rotation="-10,6.5,4", Scale="1,1,1") {
    rotationMode=BMM_Additive
    rotationSpace=BCS_BoneSpace
    translationMode=BMM_Additive
    translationSpace=BCS_BoneSpace
    alphaBoolBlend {blendOption=Linear}
    boneToModify {boneName=dyn_skirt_fr_l_MSTR}
    rotation=<rot -10 6.5 4>
    translation=<7 2 2>
   }
   n145 = Get thigh rotation_r() -> (thigh rotation_r:double)
   n146 = "Transform (Modify) Bone - Bone: dyn_skirt_fr_r_MSTR"(ComponentPose=@n144, Alpha=@n145, Translation="8,0,0", Rotation="0,0,-1", Scale="1,1,1") {
    rotationMode=BMM_Additive
    rotationSpace=BCS_BoneSpace
    translationMode=BMM_Additive
    translationSpace=BCS_BoneSpace
    alphaBoolBlend {blendOption=Linear}
    alphaScaleBiasClamp {interpSpeedDecreasing=20, interpSpeedIncreasing=20}
    boneToModify {boneName=dyn_skirt_fr_r_MSTR}
    rotation {roll=-1}
    translation {x=8}
   }
   n147 = "Transform (Modify) Bone - Bone: dyn_skirt_fr_01"(ComponentPose=@n146, Alpha=@n145, Translation="-5,0,0", Rotation="20,0,-8.039", Scale="1,1,1") {
    rotationMode=BMM_Additive
    rotationSpace=BCS_BoneSpace
    translationMode=BMM_Additive
    translationSpace=BCS_BoneSpace
    alphaBoolBlend {blendOption=Linear}
    boneToModify {boneName=dyn_skirt_fr_01}
    rotation {pitch=20, roll=-8.039}
    translation {x=-5}
   }
   n148 = Get jumping alpha() -> (jumping alpha:double)
   n149 = "Transform (Modify) Bone - Bone: dyn_skirt_fr_MSTR"(ComponentPose=@n147, Alpha=@n148, Translation="0,0,0", Rotation="-10,0.000,0.000", Scale="1,1,1") {
    rotationMode=BMM_Additive
    rotationSpace=BCS_BoneSpace
    translationMode=BMM_Additive
    translationSpace=BCS_BoneSpace
    alphaBoolBlend {blendOption=Linear}
    boneToModify {boneName=dyn_skirt_fr_MSTR}
    rotation {pitch=-10}
   }
   n150 = "Transform (Modify) Bone - Bone: dyn_skirt_fr_r_MSTR"(ComponentPose=@n149, Alpha=@n148, Translation="0,0,0", Rotation="-15,0.000,0.000", Scale="1,1,1") {
    rotationMode=BMM_Additive
    rotationSpace=BCS_BoneSpace
    translationMode=BMM_Additive
    translationSpace=BCS_BoneSpace
    alphaBoolBlend {blendOption=Linear}
    boneToModify {boneName=dyn_skirt_fr_r_MSTR}
    rotation {pitch=-15}
   }
   n151 = "Transform (Modify) Bone - Bone: dyn_skirt_fr_l_MSTR"(ComponentPose=@n150, Alpha=@n148, Translation="0,0,0", Rotation="-10,0.000,0.000", Scale="1,1,1") {
    rotationMode=BMM_Additive
    rotationSpace=BCS_BoneSpace
    translationMode=BMM_Additive
    translationSpace=BCS_BoneSpace
    alphaBoolBlend {blendOption=Linear}
    boneToModify {boneName=dyn_skirt_fr_l_MSTR}
    rotation {pitch=-10}
   }
   Save cached pose 'initial Pose'(Pose=Component To Local(ComponentPose=@n151)) {cacheName=initial Pose}
   n152 = Get gravity_override() -> (gravity_override:Vector)
   n153 = RigidBody!(ComponentPose=Local To Component(LocalPose=Use cached pose 'initial Pose'()), Alpha=0.9, OverrideWorldGravity:Vector=@n152, ExternalForce) -> (Pose) {
    bClampLinearTranslationLimitToRefPose=true
    bOverrideWorldGravity=true
    overridePhysicsAsset=&F_MED_Octoberfest_Dyn_skirt
    simulationSpace=BaseBoneSpace
    alphaBoolBlend {blendOption=Linear}
    baseBoneRef {boneName=spine_01}
    componentAppliedLinearAccClamp=<2500 2500 2500>
    componentLinearAccScale=<0.3 0.2 0.2>
    componentLinearVelScale=<0.3 0.2 0.2>
    overrideWorldGravity {z=-980}
    simSpaceSettings {maxAngularAcceleration=50, maxAngularVelocity=50, maxLinearAcceleration=50, maxLinearVelocity=50, worldAlpha=0.5}
   }
   n154 = RigidBody(ComponentPose=Local To Component(LocalPose=Use cached pose 'initial Pose'()), OverrideWorldGravity="0,0,-980") {
    bClampLinearTranslationLimitToRefPose=true
    bOverrideWorldGravity=true
    overridePhysicsAsset=&F_MED_Octoberfest_Dyn_skirt
    simulationSpace=BaseBoneSpace
    alphaBoolBlend {blendOption=Linear}
    baseBoneRef {boneName=spine_05}
    componentLinearAccScale=<2 1.2 1.2>
    componentLinearVelScale=<2 1.2 1.2>
    overrideWorldGravity {z=-980}
   }
   n155 = Look At - Bone: dyn_skirt_fr_l_MSTR(ComponentPose=@n154, Alpha=0.3) {
    alphaBoolBlend {blendInTime=0.7, blendOption=Linear, blendOutTime=0.5}
    alphaScaleBiasClamp {bInterpResult=true, clampMax=0, interpSpeedDecreasing=5, interpSpeedIncreasing=7}
    boneToModify {boneName=dyn_skirt_fr_l_MSTR}
    lookAtLocation=<-9.8928 11.0317 -3.3203>
    lookAtTarget {boneReference {boneName=thigh_l}}
    lookAt_Axis {axis=<1 0>}
   }
   n156 = Look At - Bone: dyn_skirt_fr_r_MSTR(ComponentPose=@n155, Alpha=0.3) {
    alphaBoolBlend {blendInTime=0.7, blendOption=Linear, blendOutTime=0.5}
    alphaScaleBiasClamp {bInterpResult=true, clampMax=0, interpSpeedDecreasing=5, interpSpeedIncreasing=7}
    boneToModify {boneName=dyn_skirt_fr_r_MSTR}
    lookAtLocation=<3.2692 -7.3839 1.4545>
    lookAtTarget {boneReference {boneName=thigh_r}}
    lookAt_Axis {axis=<1 0>}
   }
   n157 = Look At - Bone: dyn_skirt_out_l_MSTR(ComponentPose=@n156, Alpha=0.4) {
    alphaBoolBlend {blendInTime=0.7, blendOption=Linear, blendOutTime=0.5}
    alphaScaleBiasClamp {bInterpResult=true, clampMax=0, interpSpeedDecreasing=5, interpSpeedIncreasing=7}
    boneToModify {boneName=dyn_skirt_out_l_MSTR}
    lookAtLocation=<-10 0 -7.8201>
    lookAtTarget {boneReference {boneName=thigh_l}}
    lookAt_Axis {axis=<1 0>}
   }
   n158 = Look At - Bone: dyn_skirt_out_r_MSTR(ComponentPose=@n157, Alpha=0.4) {
    alphaBoolBlend {blendInTime=0.7, blendOption=Linear, blendOutTime=0.5}
    alphaScaleBiasClamp {bInterpResult=true, clampMax=0, interpSpeedDecreasing=5, interpSpeedIncreasing=7}
    boneToModify {boneName=dyn_skirt_out_r_MSTR}
    lookAtLocation=<10 2.5109 8.6601>
    lookAtTarget {boneReference {boneName=thigh_r}}
    lookAt_Axis {axis=<1 0>}
   }
   n159 = Blend Poses by bool(BlendPose_0=Component To Local(ComponentPose=@n158), BlendPose_1=Component To Local(ComponentPose=@n153), bActiveValue=Get bIsPlayingEmote()) {blendType=Linear}
   n160 = RigidBody(ComponentPose=Local To Component(LocalPose=Use cached pose 'initial Pose'()), OverrideWorldGravity=@n152) {
    bClampLinearTranslationLimitToRefPose=true
    bOverrideWorldGravity=true
    overridePhysicsAsset=&F_MED_Octoberfest_Dyn_skirt
    simulationSpace=BaseBoneSpace
    alphaBoolBlend {blendOption=Linear}
    baseBoneRef {boneName=spine_01}
    componentAppliedLinearAccClamp=<3000 3000 3000>
    componentLinearAccScale=<0.3 0.2 0.2>
    componentLinearVelScale=<0.3 0.2 0.2>
    overrideWorldGravity {z=-980}
   }
   n161 = Component To Local(ComponentPose=@n160)
   n162 = Get isWearingCape() -> (isWearingCape:bool)
   n163 = RigidBody!(
    ComponentPose=Local To Component(LocalPose=Blend Poses by bool(BlendPose_0=@n161, BlendPose_1=@n159, bActiveValue=@n162) {blendType=Linear})
    Alpha
    ExternalForce
   ) -> (Pose) {
    bClampLinearTranslationLimitToRefPose=true
    bOverrideWorldGravity=true
    overridePhysicsAsset=&/F_MED_Octoberfest_Dyn_torso
    simulationSpace=BaseBoneSpace
    alphaBoolBlend {blendOption=Linear}
    baseBoneRef {boneName=spine_03}
    componentAppliedLinearAccClamp=<3000 3000 3000>
    componentLinearAccScale=<0.4 0.3 0.3>
    componentLinearVelScale=<0.4 0.3 0.3>
    overrideWorldGravity {z=-980}
   }
   Save cached pose 'Rigid'(Pose=Component To Local(ComponentPose=@n163)) {cacheName=Rigid}
  ]}
 EventGraph {nodes [
   Event OnCharacterCustomizationLoaded()
   n1 = CheckForBackbling()
   > Set isWearingBackbling(isWearingBackbling=@n1)
   > Set isWearingCape(isWearingCape=@n1.bIsBackblingLong)
   Event BlueprintUpdateAnimation()
   > n2 = Sequence!(execute) -> (then_0, then_1, then_2:Exec, then_3:Exec, then_4:Exec, then_5:Exec)
   > n3 = Branch(Condition=Get bIsSwimSprinting() -> (bIsSwimSprinting:bool))
   > Set gravity_override(gravity_override="0, 0, 0")
   n4 = Branch(execute=@n3.else, Condition=Get bIsSkyDiving())
   > n5 = Branch(Condition=Get bIsParachuteOpened())
   Set gravity_override(execute=[@n4.else, @n5.then], gravity_override="0, 0,-980")
   n6 = Branch(execute=@n5.else, Condition="float > float"(A=Get PawnVelocityZ(), B=50) {is=title K2Node_CallFunction})
   > Set gravity_override(gravity_override="0, 0,200")
   Set gravity_override(execute=@n6.else, gravity_override="0, 0,-200")
   n7 = GetOwnerSkeletalMeshComponent()
   n8 = BreakRotator(InRot=BreakTransform(InTransform=GetDeltaTransformFromRefPose(self=@n7, BoneName=thigh_r, BaseName=pelvis)).Rotation)
   n9 = NOT Boolean(A=Get bIsPlayingEmote())
   n10 = Get bIsMoving2D()
   Set thigh rotation_r(
    execute=@n2.then_1
    thigh rotation_r:double=SelectFloat(A=MapRangeClamped(Value=@n8.Yaw, InRangeA=-50, InRangeB=-90), B=0, bPickA=AND Boolean(A=@n10, B=@n9))
   ) -> (Output_Get:double)
   n11 = GetOwnerSkeletalMeshComponent()
   n12 = BreakRotator(InRot=BreakTransform(InTransform=GetDeltaTransformFromRefPose(self=@n11, BoneName=thigh_l, BaseName=pelvis)).Rotation)
   n13 = NOT Boolean(A=Get bIsPlayingEmote())
   n14 = Get bIsMoving2D()
   > Set thigh_rotation_l(thigh_rotation_l:double=SelectFloat(A=MapRangeClamped(Value=@n12.Yaw, InRangeA=-50, InRangeB=-90), B=0, bPickA=AND Boolean(A=@n14, B=@n13))) -> (Output_Get:double)
   > n15 = Branch(Condition=Get bIsJumping())
   Set jumping alpha(execute=[@n15.then, @n16.then], jumping alpha:double=1) -> (Output_Get:double)
   n16 = Branch(execute=@n15.else, Condition=Get bIsFalling())
   Set jumping alpha(execute=@n16.else, jumping alpha=0)
   n17 = GetAnimInstance(self=GetOwnerSkeletalMeshComponent())
   Is Valid(exec=@n2.then_3, InputObject=@n17)
   n18 = Montage_IsPlaying(self=@n17, Montage=/BRCosmetics/Animation/Game/MainPlayer/Emotes/Mask_Off/Emote_Mask_Off_CMM)
   n19 = Montage_IsPlaying(self=@n17, Montage=/BRCosmetics/Animation/Game/MainPlayer/Emotes/Mask_Off/Emote_Mask_Off_CMF)
   > Set Snap_Skirt_front(Snap_Skirt_front:double=SelectFloat(A=1, B=0, bPickA=OR Boolean(A=@n18, B=@n19))) -> (Output_Get:double)
   n20 = NOT Boolean(A=Get bIsMoving2D())
   n21 = Get bIsCrouching()
   Set crouch alpha(execute=@n2.then_4, crouch alpha:double=SelectFloat(A=1, B=0, bPickA=AND Boolean(A=@n21, B=@n20))) -> (Output_Get:double)
   Set Crouch 0ff(execute=@n2.then_5, Crouch 0ff:double=SelectFloat(A=0, B=1, bPickA=Get bIsCrouching())) -> (Output_Get:double)
   n22 = Get PawnMovementDirectionAzimuth() -> (PawnMovementDirectionAzimuth:float)
   n23 = float - float(A=MapRangeClamped(Value=@n22, InRangeA=45, InRangeB=90), B=MapRangeClamped(Value=@n22, InRangeA=90, InRangeB=135)) {is=title K2Node_CallFunction}
   n24 = Get bIsCrouching()
   n25 = Get bIsMoving2D()
   Set Right_Direct(execute=@n2.then_2, Right_Direct:double=SelectFloat(A="Absolute (Float)"(A=@n23), B=0, bPickA=AND Boolean(A=@n24, B=@n25))) -> (Output_Get:double)
   n26 = Get PawnMovementDirectionAzimuth()
   n27 = float - float(A=MapRangeClamped(Value=@n26, InRangeA=-45, InRangeB=-90), B=MapRangeClamped(Value=@n26, InRangeA=-90, InRangeB=-135))
   n28 = Get bIsCrouching()
   n29 = Get bIsMoving2D()
   > Set Left_Direct(Left_Direct:double=SelectFloat(A="Absolute (Float)"(A=@n27), B=0, bPickA=AND Boolean(A=@n28, B=@n29))) -> (Output_Get:double)
   n30 = Get PawnMovementDirectionAzimuth()
   n31 = float - float(A=MapRangeClamped(Value=@n30, InRangeA=-1, InRangeB=0.1), B=MapRangeClamped(Value=@n30, InRangeA=0.2, InRangeB=1))
   n32 = Get bIsCrouching()
   n33 = Get bIsMoving2D()
   > Set Forward_Direct(Forward_Direct:double=SelectFloat(A="Absolute (Float)"(A=@n31), B=0, bPickA=AND Boolean(A=@n32, B=@n33))) -> (Output_Get:double)
   n34 = Branch(Condition=Get bCanPlayCustomAnimations() -> (bCanPlayCustomAnimations:bool))
   n35 = Branch(execute=[@n34.else, @n34.then], Condition=Get bIsSkyDiving())
   n36 = Branch(execute=[@n35.else, @n35.then], Condition=Get bIsSkydiveFloating())
   n37 = Get Direction Between Sockets!(
    Component:Skeletal Mesh Component Object Reference=GetOwnerSkeletalMeshComponent()
    SocketOrBoneNameFrom:Name=head
    SocketOrBoneNameTo:Name=pelvis
   ) -> (ReturnValue:Vector) {is=title K2Node_CallFunction}
   > Set skydiving vector force(skydiving vector force:Vector=vector + vector(
     A:Vector=vector * vector(A:Vector=@n37, B:Vector="75,75,75") -> (ReturnValue:Vector) {is=title K2Node_CommutativeAssociativeBinaryOperator}
     B:Vector="0, 0,2000"
    ) -> (ReturnValue:Vector) {is=title K2Node_CommutativeAssociativeBinaryOperator}) -> (Output_Get:Vector)
   n38 = Branch(execute=@n36.else, Condition=Get bIsSkydiveDiveMode())
   > Set skydiving vector force(skydiving vector force=vector + vector(A=vector * vector(A=@n37, B="180,180,200"), B="0, 0,2000"))
   Set skydiving vector force(execute=@n38.else, skydiving vector force=vector * vector(A=@n37, B="-15,-15,-15"))
  ]}
]
variables [
 Crouch 0ff {type=number}
 crouch alpha {type=number}
 Forward_Direct {type=number}
 gravity_override {type=Vector}
 isWearingBackbling {type=boolean}
 isWearingCape {type=boolean}
 jumping alpha {type=number}
 Left_Direct {type=number}
 Right_Direct {type=number}
 skydiving vector force {type=Vector}
 Snap_Skirt_front {type=number}
 thigh rotation_r {type=number}
 thigh_rotation_l {type=number}
]
